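# Animation Blueprint for the Invader pet, a hovering companion that reacts to its player's locomotion, jumps, skydiving, emotes, melee and DBNO state.
# A state machine with nested sub-machines and random breaks feeds bone modifiers for hover rotation, offsets and shrinking while DBNO-carried, then blends toward a world-space backpack target.
# The EventGraph updates these offsets each tick, retunes spring settings on crouch, melee and state notifies, and relays spark VFX and mind-control flags.
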AnimBlueprint /Game/Characters/PetsNoCarrier/Pet_M_MED_Invader/Invader_AnimBP
target_skeleton=/Game/Characters/PetsNoCarrier/Pet_M_MED_Invader/Meshes/M_MED_Invader_Pet_Skeleton
graphs [
 AnimGraph {nodes [
   "Transform (Modify) Bone - Bone: C_Root_Main_Root_Jnt"(Alpha=1, Translation="0,0,0", Rotation=Get RootRotationOffsetTotal() -> (RootRotationOffsetTotal:Rotator), Scale="1,1,1") {rotationMode=BMM_Additive, alphaBoolBlend {blendOption=Linear}, boneToModify {boneName=C_Root_Main_Root_Jnt}}
   "Transform (Modify) Bone - Bone: C_Leaf_Offset_Leaf_Jnt"(Alpha=1, Translation=Get CarrierOffsetTotal() -> (CarrierOffsetTotal:Vector), Rotation="0,0,0", Scale="1,1,1") {translationMode=BMM_Additive, alphaBoolBlend {blendOption=Linear}, boneToModify {boneName=C_Leaf_Offset_Leaf_Jnt}}
   n1 = "Transform (Modify) Bone - Bone: C_Leaf_Offset_Leaf_Jnt"(
    ComponentPose=Local To Component(LocalPose=Use cached pose 'SavedPose21014'())
    Alpha=1
    Translation=Get BackPackTargetVector() -> (BackPackTargetVector:Vector)
    Rotation="0,0,0"
    Scale="1,1,1"
   ) {
    translationMode=BMM_Replace
    translationSpace=BCS_WorldSpace
    alphaBoolBlend {blendOption=Linear}
    boneToModify {boneName=C_Leaf_Offset_Leaf_Jnt}
   }
   n2 = "Transform (Modify) Bone - Bone: C_Leaf_Offset_Leaf_Jnt"(
    ComponentPose=Local To Component(LocalPose=Use cached pose 'SavedPose21014'())
    Alpha=1
    Translation=Get CarrierOffsetTotal()
    Rotation="0,0,0"
    Scale="1,1,1"
   ) {translationMode=BMM_Additive, alphaBoolBlend {blendOption=Linear}, boneToModify {boneName=C_Leaf_Offset_Leaf_Jnt}}
   Output Pose(Result=Two Way Blend(
     A=Component To Local(ComponentPose=@n2)
     B=Component To Local(ComponentPose=@n1)
     Alpha=Get BackPackTargetVectorAlphaSmooth() -> (BackPackTargetVectorAlphaSmooth:double)
    ) {blendNode {alphaBoolBlend {blendOption=Linear}}}) {layerGroup=None}
   AnimGraphNode_StateMachine_0 = Invader Animations() {is=title AnimGraphNode_StateMachine}
   n3 = "Transform (Modify) Bone - Bone: C_Leaf_Hover_Leaf_Jnt"(
    ComponentPose=Local To Component(LocalPose=@AnimGraphNode_StateMachine_0)
    Alpha=1
    Translation="0,0,0"
    Rotation=Get HoverRotationOffsetTotal() -> (HoverRotationOffsetTotal:Rotator)
    Scale="1,1,1"
   ) {
    rotationMode=BMM_Additive
    translationSpace=BCS_WorldSpace
    alphaBoolBlend {blendOption=Linear}
    boneToModify {boneName=C_Leaf_Hover_Leaf_Jnt}
   }
   n4 = "Transform (Modify) Bone - Bone: C_Leaf_Hover_Leaf_Jnt"(
    ComponentPose=@n3
    Alpha=Get CarrierOffsetRotationVelocityAlpha() -> (CarrierOffsetRotationVelocityAlpha:double)
    Translation="0,0,0"
    Rotation=Get CarrierOffsetRotationVelocity() -> (CarrierOffsetRotationVelocity:Rotator)
    Scale="1,1,1"
   ) {
    rotationMode=BMM_Additive
    translationSpace=BCS_WorldSpace
    alphaBoolBlend {blendOption=Linear}
    boneToModify {boneName=C_Leaf_Hover_Leaf_Jnt}
   }
   n5 = "Transform (Modify) Bone - Bone: C_Leaf_Offset_Leaf_Jnt"(
    ComponentPose=@n4
    Alpha=1
    Translation="0,0,0"
    Rotation=Get RotationYawDefinedHoverOffset() -> (RotationYawDefinedHoverOffset:Rotator)
    Scale="1,1,1"
   ) {
    rotationMode=BMM_Additive
    translationSpace=BCS_WorldSpace
    alphaBoolBlend {blendOption=Linear}
    boneToModify {boneName=C_Leaf_Offset_Leaf_Jnt}
   }
   n6 = "Transform (Modify) Bone - Bone: C_Root_Main_Root_Jnt"(
    ComponentPose=@n5
    Alpha=Get IsDBNOCarriedAlpha() -> (IsDBNOCarriedAlpha:double)
    Translation="0,0,0"
    Rotation="0,0,0"
    Scale="0.01,0.01,0.01"
   ) {
    scaleMode=BMM_Replace
    scaleSpace=BCS_WorldSpace
    translationMode=BMM_Replace
    alphaBoolBlend {blendOption=Linear}
    boneToModify {boneName=C_Root_Main_Root_Jnt}
    scale=<0.01 0.01 0.01>
   }
   Save cached pose 'SavedPose21014'(Pose=Component To Local(ComponentPose=@n6)) {cacheName=SavedPose21014}
  ]}
 Invader Animations {
  entry=Locomotion
  path=AnimGraph.AnimGraphNode_StateMachine_0.Invader Animations
  ids=id
  nodes [
   AnimStateNode_0 = Locomotion() {is=title AnimStateNode, stateFullyBlended {guid=#1, notifyName=Locomotion_FullyBlended}}
   AnimStateNode_1 = Random Breaks() {
    is=title AnimStateNode
    stateEntered {guid=#2, notifyName=RandomBreaks_Entered}
    stateLeft {guid=#3, notifyName=RandomBreaks_Exited}
   }
   AnimStateNode_10 = Emote_React_Loop() {is=title AnimStateNode}
   AnimStateNode_11 = Skydive() {is=title AnimStateNode}
   AnimStateNode_2 = Jump() {is=title AnimStateNode}
   AnimStateNode_3 = JumpLand() {is=title AnimStateNode}
   AnimStateNode_4 = DBNO() {is=title AnimStateNode, stateEntered {guid=#4, notifyName=DBNO_Entered}}
   AnimStateNode_5 = DBNO_End() {is=title AnimStateNode, stateLeft {guid=#5, notifyName=DBNO_Exited}}
   AnimStateNode_6 = Skydive_Dive() {is=title AnimStateNode, bAlwaysResetOnEntry=true}
   AnimStateNode_7 = Hunker() {
    is=title AnimStateNode
    stateEntered {guid=#6, notifyName=Hunker_Entered}
    stateLeft {guid=#7, notifyName=Hunker_Exited}
   }
   AnimStateNode_8 = Skydive_Dive_Stop() {is=title AnimStateNode}
   AnimStateNode_9 = Emote_React() {is=title AnimStateNode}
   AnimStateTransitionNode_0 = AnimStateTransitionNode() {crossfadeDuration=0.4}
   AnimStateTransitionNode_1 = AnimStateTransitionNode()
   AnimStateTransitionNode_10 = AnimStateTransitionNode()
   AnimStateTransitionNode_11 = AnimStateTransitionNode()
   AnimStateTransitionNode_12 = AnimStateTransitionNode()
   AnimStateTransitionNode_13 = AnimStateTransitionNode() {bAutomaticRuleBasedOnSequencePlayerInState=true, crossfadeDuration=0.4}
   AnimStateTransitionNode_14 = AnimStateTransitionNode()
   AnimStateTransitionNode_15 = AnimStateTransitionNode()
   AnimStateTransitionNode_17 = AnimStateTransitionNode() {crossfadeDuration=0.4}
   AnimStateTransitionNode_18 = AnimStateTransitionNode() {bAutomaticRuleBasedOnSequencePlayerInState=true}
   AnimStateTransitionNode_19 = AnimStateTransitionNode()
   AnimStateTransitionNode_2 = AnimStateTransitionNode()
   AnimStateTransitionNode_20 = AnimStateTransitionNode()
   AnimStateTransitionNode_21 = AnimStateTransitionNode() {bAutomaticRuleBasedOnSequencePlayerInState=true}
   AnimStateTransitionNode_22 = AnimStateTransitionNode()
   AnimStateTransitionNode_3 = AnimStateTransitionNode() {bAutomaticRuleBasedOnSequencePlayerInState=true, crossfadeDuration=0.5}
   AnimStateTransitionNode_4 = AnimStateTransitionNode()
   AnimStateTransitionNode_5 = AnimStateTransitionNode() {crossfadeDuration=0.4}
   AnimStateTransitionNode_6 = AnimStateTransitionNode()
   AnimStateTransitionNode_7 = AnimStateTransitionNode()
   AnimStateTransitionNode_8 = AnimStateTransitionNode()
   AnimStateTransitionNode_9 = AnimStateTransitionNode() {bAutomaticRuleBasedOnSequencePlayerInState=true, crossfadeDuration=0.8}
  ]
  links [
   {
    dst=DBNO_End
    rule=AnimGraph.AnimGraphNode_StateMachine_0.Invader Animations.AnimStateTransitionNode_8.Transition
    src=DBNO
    via=AnimStateTransitionNode_8
   }
   {
    dst=Locomotion
    rule=AnimGraph.AnimGraphNode_StateMachine_0.Invader Animations.AnimStateTransitionNode_9.Transition
    src=DBNO_End
    via=AnimStateTransitionNode_9
   }
   {
    dst=Emote_React_Loop
    rule=AnimGraph.AnimGraphNode_StateMachine_0.Invader Animations.AnimStateTransitionNode_18.Transition
    src=Emote_React
    via=AnimStateTransitionNode_18
   }
   {
    dst=Locomotion
    rule=AnimGraph.AnimGraphNode_StateMachine_0.Invader Animations.AnimStateTransitionNode_19.Transition
    src=Emote_React_Loop
    via=AnimStateTransitionNode_19
   }
   {
    dst=Locomotion
    rule=AnimGraph.AnimGraphNode_StateMachine_0.Invader Animations.AnimStateTransitionNode_5.Transition
    src=Hunker
    via=AnimStateTransitionNode_5
   }
   {
    dst=JumpLand
    rule=AnimGraph.AnimGraphNode_StateMachine_0.Invader Animations.AnimStateTransitionNode_1.Transition
    src=Jump
    via=AnimStateTransitionNode_1
   }
   {
    dst=JumpLand
    rule=AnimGraph.AnimGraphNode_StateMachine_0.Invader Animations.AnimStateTransitionNode_2.Transition
    src=Jump
    via=AnimStateTransitionNode_2
   }
   {
    dst=Jump
    rule=AnimGraph.AnimGraphNode_StateMachine_0.Invader Animations.AnimStateTransitionNode_4.Transition
    src=JumpLand
    via=AnimStateTransitionNode_4
   }
   {
    dst=Locomotion
    rule=AnimGraph.AnimGraphNode_StateMachine_0.Invader Animations.AnimStateTransitionNode_3.Transition
    src=JumpLand
    via=AnimStateTransitionNode_3
   }
   {
    dst=Locomotion
    rule=AnimGraph.AnimGraphNode_StateMachine_0.Invader Animations.AnimStateTransitionNode_6.Transition
    src=JumpLand
    via=AnimStateTransitionNode_6
   }
   {
    dst=DBNO
    rule=AnimGraph.AnimGraphNode_StateMachine_0.Invader Animations.AnimStateTransitionNode_7.Transition
    src=Locomotion
    via=AnimStateTransitionNode_7
   }
   {
    dst=Emote_React
    rule=AnimGraph.AnimGraphNode_StateMachine_0.Invader Animations.AnimStateTransitionNode_11.Transition
    src=Locomotion
    via=AnimStateTransitionNode_11
   }
   {
    dst=Hunker
    rule=AnimGraph.AnimGraphNode_StateMachine_0.Invader Animations.AnimStateTransitionNode_10.Transition
    src=Locomotion
    via=AnimStateTransitionNode_10
   }
   {
    dst=Jump
    rule=AnimGraph.AnimGraphNode_StateMachine_0.Invader Animations.AnimStateTransitionNode_0.Transition
    src=Locomotion
    via=AnimStateTransitionNode_0
   }
   {
    dst=Random Breaks
    rule=AnimGraph.AnimGraphNode_StateMachine_0.Invader Animations.AnimStateTransitionNode_12.Transition
    src=Locomotion
    via=AnimStateTransitionNode_12
   }
   {
    dst=Skydive
    rule=AnimGraph.AnimGraphNode_StateMachine_0.Invader Animations.AnimStateTransitionNode_17.Transition
    src=Locomotion
    via=AnimStateTransitionNode_17
   }
   {
    dst=Locomotion
    rule=AnimGraph.AnimGraphNode_StateMachine_0.Invader Animations.AnimStateTransitionNode_13.Transition
    src=Random Breaks
    via=AnimStateTransitionNode_13
   }
   {
    dst=Locomotion
    rule=AnimGraph.AnimGraphNode_StateMachine_0.Invader Animations.AnimStateTransitionNode_14.Transition
    src=Random Breaks
    via=AnimStateTransitionNode_14
   }
   {
    dst=Locomotion
    rule=AnimGraph.AnimGraphNode_StateMachine_0.Invader Animations.AnimStateTransitionNode_22.Transition
    src=Skydive
    via=AnimStateTransitionNode_22
   }
   {
    dst=Skydive_Dive
    rule=AnimGraph.AnimGraphNode_StateMachine_0.Invader Animations.AnimStateTransitionNode_20.AnimationTransitionGraph_2
    src=Skydive
    via=AnimStateTransitionNode_20
   }
   {
    dst=Skydive_Dive_Stop
    rule=AnimGraph.AnimGraphNode_StateMachine_0.Invader Animations.AnimStateTransitionNode_15.AnimationTransitionGraph_1
    src=Skydive_Dive
    via=AnimStateTransitionNode_15
   }
   {
    dst=Skydive
    rule=AnimGraph.AnimGraphNode_StateMachine_0.Invader Animations.AnimStateTransitionNode_21.AnimationTransitionGraph_2
    src=Skydive_Dive_Stop
    via=AnimStateTransitionNode_21
   }
  ]
 }
 Locomotion {
  path=AnimGraph.AnimGraphNode_StateMachine_0.Invader Animations.AnimStateNode_0.Locomotion
  nodes [
   AnimGraphNode_StateMachine_0 = Locomotion() {is=title AnimGraphNode_StateMachine}
   Output Animation Pose(Result=@AnimGraphNode_StateMachine_0)
  ]
 }
 Locomotion {
  entry=Idle
  path=AnimGraph.AnimGraphNode_StateMachine_0.Invader Animations.AnimStateNode_0.Locomotion.AnimGraphNode_StateMachine_0.Locomotion
  ids=id
  nodes [
   AnimStateNode_0 = Locomotion() {
    is=title AnimStateNode
    stateEntered {guid=#8}
    stateFullyBlended {guid=#9}
    stateLeft {guid=#10}
   }
   AnimStateNode_2 = Idle() {
    stateEntered {guid=#11, notifyName=Idle_Entered}
    stateLeft {guid=#12, notifyName=Idle_Exited}
   }
   AnimStateTransitionNode_3 = AnimStateTransitionNode() {crossfadeDuration=1}
   AnimStateTransitionNode_4 = AnimStateTransitionNode()
  ]
  links [
   {
    dst=Locomotion
    rule=AnimGraph.AnimGraphNode_StateMachine_0.Invader Animations.AnimStateNode_0.Locomotion.AnimGraphNode_StateMachine_0.Locomotion.AnimStateTransitionNode_4.Transition
    src=Idle
    via=AnimStateTransitionNode_4
   }
   {
    dst=Idle
    rule=AnimGraph.AnimGraphNode_StateMachine_0.Invader Animations.AnimStateNode_0.Locomotion.AnimGraphNode_StateMachine_0.Locomotion.AnimStateTransitionNode_3.Transition
    src=Locomotion
    via=AnimStateTransitionNode_3
   }
  ]
 }
 Locomotion {
  path=AnimGraph.AnimGraphNode_StateMachine_0.Invader Animations.AnimStateNode_0.Locomotion.AnimGraphNode_StateMachine_0.Locomotion.AnimStateNode_0.Locomotion
  nodes [
   Output Animation Pose(Result=Blendspace Player 'Invader_Locomotion_BS'(X=Get Local Velocity Yaw Angle() -> (Local Velocity Yaw Angle:double), Y=0) {blendSpace=/Game/Animation/Game/Pets/Invader_Merged/Backpack/Invader_Locomotion_BS})
   Get!() -> (PawnVelocity_X:double, PawnVelocity_Y:double, PawnVelocity_Z:double)
  ]
 }
 Idle {
  path=AnimGraph.AnimGraphNode_StateMachine_0.Invader Animations.AnimStateNode_0.Locomotion.AnimGraphNode_StateMachine_0.Locomotion.AnimStateNode_2.Idle
  nodes [Output Animation Pose(Result=Sequence Player 'Hover_Idle_NonTargeted'() {sequence=/Game/Animation/Game/Pets/Invader_Merged/Backpack/Hover_Idle_NonTargeted})]
 }
 Transition {
  path=AnimGraph.AnimGraphNode_StateMachine_0.Invader Animations.AnimStateNode_0.Locomotion.AnimGraphNode_StateMachine_0.Locomotion.AnimStateTransitionNode_3.Transition
  nodes [Result(bCanEnterTransition=NOT Boolean(A=Get Is Moving 2D() -> (Is Moving 2D:bool)))]
 }
 Transition {
  path=AnimGraph.AnimGraphNode_StateMachine_0.Invader Animations.AnimStateNode_0.Locomotion.AnimGraphNode_StateMachine_0.Locomotion.AnimStateTransitionNode_4.Transition
  nodes [Result(bCanEnterTransition=Get Is Moving 2D())]
 }
 Random Breaks {
  path=AnimGraph.AnimGraphNode_StateMachine_0.Invader Animations.AnimStateNode_1.Random Breaks
  nodes [
   n1 = Blend Poses by int!(
    BlendPose_0:Pose Link Structure=Sequence Player 'LocomotionBreak1'!(bLoopAnimation:bool=False) -> (Pose) {bLoopAnimation=false, sequence=/Game/Animation/Game/Pets/Invader_Merged/Backpack/LocomotionBreak1}
    BlendPose_1:Pose Link Structure=Sequence Player 'Happy'!(bLoopAnimation:bool=False) -> (Pose) {bLoopAnimation=false, sequence=/Game/Animation/Game/Pets/Invader_Merged/Backpack/Happy}
    BlendTime_0:float=0.1
    BlendTime_1:float=0.1
    ActiveChildIndex:int=Get Choose Loco Break Index() -> (Choose Loco Break Index:int=0)
   ) -> (Pose:Pose Link Structure) {
    is=title AnimGraphNode_BlendListByInt
    blendType=Linear
    blendPose [{linkId=-1, sourceLinkId=-1}, {linkId=-1, sourceLinkId=-1}]
    blendTime [0.1,0.1]
   }
   Output Animation Pose(Result=@n1)
  ]
 }
 Emote_React_Loop {
  path=AnimGraph.AnimGraphNode_StateMachine_0.Invader Animations.AnimStateNode_10.Emote_React_Loop
  nodes [Output Animation Pose(Result=Sequence Player 'EmoteReact_Loop'() {sequence=/Game/Animation/Game/Pets/Invader_Merged/Backpack/EmoteReact_Loop})]
 }
 Skydive {
  path=AnimGraph.AnimGraphNode_StateMachine_0.Invader Animations.AnimStateNode_11.Skydive
  nodes [
   AnimGraphNode_StateMachine_0 = SkyDive() {is=title AnimGraphNode_StateMachine}
   Output Animation Pose(Result=@AnimGraphNode_StateMachine_0)
  ]
 }
 SkyDive {
  entry=Skydive
  path=AnimGraph.AnimGraphNode_StateMachine_0.Invader Animations.AnimStateNode_11.Skydive.AnimGraphNode_StateMachine_0.SkyDive
  ids=id
  nodes [
   AnimStateNode_11 = Skydive()
   AnimStateNode_6 = Skydive_Dive() {bAlwaysResetOnEntry=true}
   AnimStateNode_8 = Skydive_Dive_Stop()
   AnimStateTransitionNode_15 = AnimStateTransitionNode()
   AnimStateTransitionNode_20 = AnimStateTransitionNode()
   AnimStateTransitionNode_21 = AnimStateTransitionNode() {bAutomaticRuleBasedOnSequencePlayerInState=true}
  ]
  links [
   {
    dst=Skydive_Dive
    rule=AnimGraph.AnimGraphNode_StateMachine_0.Invader Animations.AnimStateNode_11.Skydive.AnimGraphNode_StateMachine_0.SkyDive.AnimStateTransitionNode_20.AnimationTransitionGraph_2
    src=Skydive
    via=AnimStateTransitionNode_20
   }
   {
    dst=Skydive_Dive_Stop
    rule=AnimGraph.AnimGraphNode_StateMachine_0.Invader Animations.AnimStateNode_11.Skydive.AnimGraphNode_StateMachine_0.SkyDive.AnimStateTransitionNode_15.AnimationTransitionGraph_1
    src=Skydive_Dive
    via=AnimStateTransitionNode_15
   }
   {
    dst=Skydive
    rule=AnimGraph.AnimGraphNode_StateMachine_0.Invader Animations.AnimStateNode_11.Skydive.AnimGraphNode_StateMachine_0.SkyDive.AnimStateTransitionNode_21.AnimationTransitionGraph_2
    src=Skydive_Dive_Stop
    via=AnimStateTransitionNode_21
   }
  ]
 }
 Skydive {
  path=AnimGraph.AnimGraphNode_StateMachine_0.Invader Animations.AnimStateNode_11.Skydive.AnimGraphNode_StateMachine_0.SkyDive.AnimStateNode_11.Skydive
  nodes [Output Animation Pose(Result=Sequence Player 'SkyDive'() {sequence=/Game/Animation/Game/Pets/Invader_Merged/Backpack/SkyDive})]
 }
 Skydive_Dive {
  path=AnimGraph.AnimGraphNode_StateMachine_0.Invader Animations.AnimStateNode_11.Skydive.AnimGraphNode_StateMachine_0.SkyDive.AnimStateNode_6.Skydive_Dive
  nodes [
   AnimGraphNode_StateMachine_0 = Skydive_Dive() {is=title AnimGraphNode_StateMachine}
   Output Animation Pose(Result=@AnimGraphNode_StateMachine_0)
  ]
 }
 Skydive_Dive {
  entry=Skydive_Dive_Start
  path=AnimGraph.AnimGraphNode_StateMachine_0.Invader Animations.AnimStateNode_11.Skydive.AnimGraphNode_StateMachine_0.SkyDive.AnimStateNode_6.Skydive_Dive.AnimGraphNode_StateMachine_0.Skydive_Dive
  ids=id
  nodes [
   AnimStateNode_0 = Skydive_Dive_Start() {is=title AnimStateNode}
   AnimStateNode_1 = Skydive_Dive_Loop() {is=title AnimStateNode}
   AnimStateTransitionNode_0 = AnimStateTransitionNode() {bAutomaticRuleBasedOnSequencePlayerInState=true, crossfadeDuration=0}
  ]
  links [{
    dst=Skydive_Dive_Loop
    rule=AnimGraph.AnimGraphNode_StateMachine_0.Invader Animations.AnimStateNode_11.Skydive.AnimGraphNode_StateMachine_0.SkyDive.AnimStateNode_6.Skydive_Dive.AnimGraphNode_StateMachine_0.Skydive_Dive.AnimStateTransitionNode_0.Transition
    src=Skydive_Dive_Start
    via=AnimStateTransitionNode_0
   }]
 }
 Skydive_Dive_Start {
  path=AnimGraph.AnimGraphNode_StateMachine_0.Invader Animations.AnimStateNode_11.Skydive.AnimGraphNode_StateMachine_0.SkyDive.AnimStateNode_6.Skydive_Dive.AnimGraphNode_StateMachine_0.Skydive_Dive.AnimStateNode_0.Skydive_Dive_Start
  nodes [Output Animation Pose(Result=Sequence Player 'SkyDive_Dive_Start'!(bLoopAnimation:bool=False) -> (Pose) {bLoopAnimation=false, sequence=/Game/Animation/Game/Pets/Invader_Merged/Backpack/SkyDive_Dive_Start})]
 }
 Skydive_Dive_Loop {
  path=AnimGraph.AnimGraphNode_StateMachine_0.Invader Animations.AnimStateNode_11.Skydive.AnimGraphNode_StateMachine_0.SkyDive.AnimStateNode_6.Skydive_Dive.AnimGraphNode_StateMachine_0.Skydive_Dive.AnimStateNode_1.Skydive_Dive_Loop
  nodes [Output Animation Pose(Result=Sequence Player 'SkyDive_Dive_B'() {sequence=/Game/Animation/Game/Pets/Invader_Merged/Backpack/SkyDive_Dive_B})]
 }
 Transition {
  path=AnimGraph.AnimGraphNode_StateMachine_0.Invader Animations.AnimStateNode_11.Skydive.AnimGraphNode_StateMachine_0.SkyDive.AnimStateNode_6.Skydive_Dive.AnimGraphNode_StateMachine_0.Skydive_Dive.AnimStateTransitionNode_0.Transition
  nodes [Result()]
 }
 Skydive_Dive_Stop {
  path=AnimGraph.AnimGraphNode_StateMachine_0.Invader Animations.AnimStateNode_11.Skydive.AnimGraphNode_StateMachine_0.SkyDive.AnimStateNode_8.Skydive_Dive_Stop
  nodes [Output Animation Pose(Result=Sequence Player 'SkyDive_Dive_Stop'!(bLoopAnimation:bool=False) -> (Pose) {bLoopAnimation=false, sequence=/Game/Animation/Game/Pets/Invader_Merged/Backpack/SkyDive_Dive_Stop})]
 }
 AnimationTransitionGraph_1 {
  path=AnimGraph.AnimGraphNode_StateMachine_0.Invader Animations.AnimStateNode_11.Skydive.AnimGraphNode_StateMachine_0.SkyDive.AnimStateTransitionNode_15.AnimationTransitionGraph_1
  nodes [Result(bCanEnterTransition=NOT Boolean(A=Get Skydiving Is Skydive Dive Mode() -> (Skydiving Is Skydive Dive Mode:bool)))]
 }
 AnimationTransitionGraph_2 {
  path=AnimGraph.AnimGraphNode_StateMachine_0.Invader Animations.AnimStateNode_11.Skydive.AnimGraphNode_StateMachine_0.SkyDive.AnimStateTransitionNode_20.AnimationTransitionGraph_2
  nodes [Result(bCanEnterTransition=Get Skydiving Is Skydive Dive Mode())]
 }
 AnimationTransitionGraph_2 {
  path=AnimGraph.AnimGraphNode_StateMachine_0.Invader Animations.AnimStateNode_11.Skydive.AnimGraphNode_StateMachine_0.SkyDive.AnimStateTransitionNode_21.AnimationTransitionGraph_2
  nodes [Result()]
 }
 Jump {
  path=AnimGraph.AnimGraphNode_StateMachine_0.Invader Animations.AnimStateNode_2.Jump
  nodes [
   AnimGraphNode_StateMachine_0 = Jump_SM() {is=title AnimGraphNode_StateMachine}
   Output Animation Pose(Result=@AnimGraphNode_StateMachine_0)
  ]
 }
 Jump_SM {
  entry=Jump_Start
  path=AnimGraph.AnimGraphNode_StateMachine_0.Invader Animations.AnimStateNode_2.Jump.AnimGraphNode_StateMachine_0.Jump_SM
  ids=id
  nodes [
   AnimStateNode_0 = Jump_Start() {is=title AnimStateNode}
   AnimStateNode_1 = Jump_Loop() {is=title AnimStateNode}
   AnimStateTransitionNode_3 = AnimStateTransitionNode() {bAutomaticRuleBasedOnSequencePlayerInState=true}
   AnimStateTransitionNode_4 = AnimStateTransitionNode() {crossfadeDuration=0.5}
  ]
  links [
   {
    dst=Jump_Start
    rule=AnimGraph.AnimGraphNode_StateMachine_0.Invader Animations.AnimStateNode_2.Jump.AnimGraphNode_StateMachine_0.Jump_SM.AnimStateTransitionNode_4.Transition
    src=Jump_Loop
    via=AnimStateTransitionNode_4
   }
   {
    dst=Jump_Loop
    rule=AnimGraph.AnimGraphNode_StateMachine_0.Invader Animations.AnimStateNode_2.Jump.AnimGraphNode_StateMachine_0.Jump_SM.AnimStateTransitionNode_3.Transition
    src=Jump_Start
    via=AnimStateTransitionNode_3
   }
  ]
 }
 Jump_Start {
  path=AnimGraph.AnimGraphNode_StateMachine_0.Invader Animations.AnimStateNode_2.Jump.AnimGraphNode_StateMachine_0.Jump_SM.AnimStateNode_0.Jump_Start
  nodes [Output Animation Pose(Result=Sequence Player 'Jump_Start'() {bLoopAnimation=false, sequence=/Game/Animation/Game/Pets/Invader_Merged/Backpack/Jump_Start})]
 }
 Jump_Loop {
  path=AnimGraph.AnimGraphNode_StateMachine_0.Invader Animations.AnimStateNode_2.Jump.AnimGraphNode_StateMachine_0.Jump_SM.AnimStateNode_1.Jump_Loop
  nodes [Output Animation Pose(Result=Sequence Player 'Jump_Loop'() {sequence=/Game/Animation/Game/Pets/Invader_Merged/Backpack/Jump_Loop})]
 }
 Transition {
  path=AnimGraph.AnimGraphNode_StateMachine_0.Invader Animations.AnimStateNode_2.Jump.AnimGraphNode_StateMachine_0.Jump_SM.AnimStateTransitionNode_3.Transition
  nodes [Result()]
 }
 Transition {
  path=AnimGraph.AnimGraphNode_StateMachine_0.Invader Animations.AnimStateNode_2.Jump.AnimGraphNode_StateMachine_0.Jump_SM.AnimStateTransitionNode_4.Transition
  nodes [Result(bCanEnterTransition=Get Is Jumping() -> (Is Jumping:bool))]
 }
 JumpLand {
  path=AnimGraph.AnimGraphNode_StateMachine_0.Invader Animations.AnimStateNode_3.JumpLand
  nodes [Output Animation Pose(Result=Sequence Player 'Jump_Land'() {sequence=/Game/Animation/Game/Pets/Invader_Merged/Backpack/Jump_Land})]
 }
 DBNO {
  path=AnimGraph.AnimGraphNode_StateMachine_0.Invader Animations.AnimStateNode_4.DBNO
  nodes [
   AnimGraphNode_StateMachine_0 = DBNO_SM() {is=title AnimGraphNode_StateMachine}
   Output Animation Pose(Result=@AnimGraphNode_StateMachine_0)
  ]
 }
 DBNO_SM {
  entry=DBNO_Start
  path=AnimGraph.AnimGraphNode_StateMachine_0.Invader Animations.AnimStateNode_4.DBNO.AnimGraphNode_StateMachine_0.DBNO_SM
  ids=id
  nodes [
   AnimStateNode_0 = DBNO_Start() {is=title AnimStateNode}
   AnimStateNode_1 = DBNO_Loop() {is=title AnimStateNode}
   AnimStateTransitionNode_3 = AnimStateTransitionNode() {bAutomaticRuleBasedOnSequencePlayerInState=true}
  ]
  links [{
    dst=DBNO_Loop
    rule=AnimGraph.AnimGraphNode_StateMachine_0.Invader Animations.AnimStateNode_4.DBNO.AnimGraphNode_StateMachine_0.DBNO_SM.AnimStateTransitionNode_3.Transition
    src=DBNO_Start
    via=AnimStateTransitionNode_3
   }]
 }
 DBNO_Start {
  path=AnimGraph.AnimGraphNode_StateMachine_0.Invader Animations.AnimStateNode_4.DBNO.AnimGraphNode_StateMachine_0.DBNO_SM.AnimStateNode_0.DBNO_Start
  nodes [Output Animation Pose(Result=Sequence Player 'DBNO_Start'() {bLoopAnimation=false, sequence=/Game/Animation/Game/Pets/Invader_Merged/Backpack/DBNO_Start})]
 }
 DBNO_Loop {
  path=AnimGraph.AnimGraphNode_StateMachine_0.Invader Animations.AnimStateNode_4.DBNO.AnimGraphNode_StateMachine_0.DBNO_SM.AnimStateNode_1.DBNO_Loop
  nodes [Output Animation Pose(Result=Sequence Player 'DBNO_Loop'() {sequence=/Game/Animation/Game/Pets/Invader_Merged/Backpack/DBNO_Loop})]
 }
 Transition {
  path=AnimGraph.AnimGraphNode_StateMachine_0.Invader Animations.AnimStateNode_4.DBNO.AnimGraphNode_StateMachine_0.DBNO_SM.AnimStateTransitionNode_3.Transition
  nodes [Result()]
 }
 DBNO_End {
  path=AnimGraph.AnimGraphNode_StateMachine_0.Invader Animations.AnimStateNode_5.DBNO_End
  nodes [Output Animation Pose(Result=Sequence Player 'DBNO_End'() {bLoopAnimation=false, sequence=/Game/Animation/Game/Pets/Invader_Merged/Backpack/DBNO_End})]
 }
 Skydive_Dive {
  path=AnimGraph.AnimGraphNode_StateMachine_0.Invader Animations.AnimStateNode_6.Skydive_Dive
  nodes [
   AnimGraphNode_StateMachine_0 = Skydive_Dive()
   Output Animation Pose(Result=@AnimGraphNode_StateMachine_0)
  ]
 }
 Skydive_Dive {
  entry=Skydive_Dive_Start
  path=AnimGraph.AnimGraphNode_StateMachine_0.Invader Animations.AnimStateNode_6.Skydive_Dive.AnimGraphNode_StateMachine_0.Skydive_Dive
  ids=id
  nodes [
   AnimStateNode_0 = Skydive_Dive_Start()
   AnimStateNode_1 = Skydive_Dive_Loop()
   AnimStateTransitionNode_0 = AnimStateTransitionNode() {bAutomaticRuleBasedOnSequencePlayerInState=true, crossfadeDuration=0}
  ]
  links [{
    dst=Skydive_Dive_Loop
    rule=AnimGraph.AnimGraphNode_StateMachine_0.Invader Animations.AnimStateNode_6.Skydive_Dive.AnimGraphNode_StateMachine_0.Skydive_Dive.AnimStateTransitionNode_0.Transition
    src=Skydive_Dive_Start
    via=AnimStateTransitionNode_0
   }]
 }
 Skydive_Dive_Start {
  path=AnimGraph.AnimGraphNode_StateMachine_0.Invader Animations.AnimStateNode_6.Skydive_Dive.AnimGraphNode_StateMachine_0.Skydive_Dive.AnimStateNode_0.Skydive_Dive_Start
  nodes [Output Animation Pose(Result=Sequence Player 'SkyDive_Dive_Start'(bLoopAnimation=False) {bLoopAnimation=false, sequence=&SkyDive_Dive_Start})]
 }
 Skydive_Dive_Loop {
  path=AnimGraph.AnimGraphNode_StateMachine_0.Invader Animations.AnimStateNode_6.Skydive_Dive.AnimGraphNode_StateMachine_0.Skydive_Dive.AnimStateNode_1.Skydive_Dive_Loop
  nodes [Output Animation Pose(Result=Sequence Player 'SkyDive_Dive_B'() {sequence=&SkyDive_Dive_B})]
 }
 Transition {
  path=AnimGraph.AnimGraphNode_StateMachine_0.Invader Animations.AnimStateNode_6.Skydive_Dive.AnimGraphNode_StateMachine_0.Skydive_Dive.AnimStateTransitionNode_0.Transition
  nodes [Result()]
 }
 Hunker {
  path=AnimGraph.AnimGraphNode_StateMachine_0.Invader Animations.AnimStateNode_7.Hunker
  nodes [
   Local Space Ref Pose()
   Sequence Player 'Hover_Idle_NonTargeted'() {sequence=&Hover_Idle_NonTargeted}
   Sequence Player 'Hunker'() {sequence=/Game/Animation/Game/Pets/Invader_Merged/Backpack/Hunker}
   Output Animation Pose(Result=Sequence Player 'HunkerPose'() {sequence=/Game/Animation/Game/Pets/Invader_Merged/Backpack/HunkerPose})
  ]
 }
 Skydive_Dive_Stop {
  path=AnimGraph.AnimGraphNode_StateMachine_0.Invader Animations.AnimStateNode_8.Skydive_Dive_Stop
  nodes [Output Animation Pose(Result=Sequence Player 'SkyDive_Dive_Stop'(bLoopAnimation=False) {bLoopAnimation=false, sequence=&SkyDive_Dive_Stop})]
 }
 Emote_React {
  path=AnimGraph.AnimGraphNode_StateMachine_0.Invader Animations.AnimStateNode_9.Emote_React
  nodes [Output Animation Pose(Result=Sequence Player 'EmoteReact_Start'() {sequence=/Game/Animation/Game/Pets/Invader_Merged/Backpack/EmoteReact_Start})]
 }
 Transition {
  path=AnimGraph.AnimGraphNode_StateMachine_0.Invader Animations.AnimStateTransitionNode_0.Transition
  nodes [Result(bCanEnterTransition=Get Is Jumping())]
 }
 Transition {
  path=AnimGraph.AnimGraphNode_StateMachine_0.Invader Animations.AnimStateTransitionNode_1.Transition
  nodes [Result(bCanEnterTransition=Get IsMovingOnGround() -> (IsMovingOnGround:bool))]
 }
 Transition {
  path=AnimGraph.AnimGraphNode_StateMachine_0.Invader Animations.AnimStateTransitionNode_10.Transition
  nodes [Result(bCanEnterTransition=Get Is Playing Melee Anim() -> (Is Playing Melee Anim:bool))]
 }
 Transition {
  path=AnimGraph.AnimGraphNode_StateMachine_0.Invader Animations.AnimStateTransitionNode_11.Transition
  nodes [Result(bCanEnterTransition=Get Is Playing Emote() -> (Is Playing Emote:bool))]
 }
 Transition {
  path=AnimGraph.AnimGraphNode_StateMachine_0.Invader Animations.AnimStateTransitionNode_12.Transition
  nodes [Result(bCanEnterTransition=Get Allow Locomotion Break() -> (Allow Locomotion Break:bool))]
 }
 Transition {path=AnimGraph.AnimGraphNode_StateMachine_0.Invader Animations.AnimStateTransitionNode_13.Transition, nodes [Result()]}
 Transition {
  path=AnimGraph.AnimGraphNode_StateMachine_0.Invader Animations.AnimStateTransitionNode_14.Transition
  nodes [Result(bCanEnterTransition=OR Boolean(A=Get Is Playing Melee Anim(), B=Get Is DBNO() -> (Is DBNO:bool)))]
 }
 AnimationTransitionGraph_1 {
  path=AnimGraph.AnimGraphNode_StateMachine_0.Invader Animations.AnimStateTransitionNode_15.AnimationTransitionGraph_1
  nodes [Result(bCanEnterTransition=NOT Boolean(A=Get Skydiving Is Skydive Dive Mode()))]
 }
 Transition {
  path=AnimGraph.AnimGraphNode_StateMachine_0.Invader Animations.AnimStateTransitionNode_17.Transition
  nodes [
   n1 = Get Skydiving Is Parachute Open() -> (Skydiving Is Parachute Open:bool)
   Result(bCanEnterTransition=AND Boolean(A=NOT Boolean(A=@n1), B=Get Skydiving Is Skydiving() -> (Skydiving Is Skydiving:bool)))
  ]
 }
 Transition {path=AnimGraph.AnimGraphNode_StateMachine_0.Invader Animations.AnimStateTransitionNode_18.Transition, nodes [Result()]}
 Transition {
  path=AnimGraph.AnimGraphNode_StateMachine_0.Invader Animations.AnimStateTransitionNode_19.Transition
  nodes [
   n1 = Get Is Playing Emote()
   Result(bCanEnterTransition=OR Boolean(A=NOT Boolean(A=@n1), B=Get Is DBNO()))
  ]
 }
 Transition {
  path=AnimGraph.AnimGraphNode_StateMachine_0.Invader Animations.AnimStateTransitionNode_2.Transition
  nodes [Result(bCanEnterTransition=Get Is Surface Swimming() -> (Is Surface Swimming:bool))]
 }
 AnimationTransitionGraph_2 {
  path=AnimGraph.AnimGraphNode_StateMachine_0.Invader Animations.AnimStateTransitionNode_20.AnimationTransitionGraph_2
  nodes [Result(bCanEnterTransition=Get Skydiving Is Skydive Dive Mode())]
 }
 AnimationTransitionGraph_2 {path=AnimGraph.AnimGraphNode_StateMachine_0.Invader Animations.AnimStateTransitionNode_21.AnimationTransitionGraph_2, nodes [Result()]}
 Transition {
  path=AnimGraph.AnimGraphNode_StateMachine_0.Invader Animations.AnimStateTransitionNode_22.Transition
  nodes [
   n1 = Get Skydiving Is Skydiving()
   Result(bCanEnterTransition=OR Boolean(A=Get Skydiving Is Parachute Open(), B=NOT Boolean(A=@n1)))
  ]
 }
 Transition {path=AnimGraph.AnimGraphNode_StateMachine_0.Invader Animations.AnimStateTransitionNode_3.Transition, nodes [Result()]}
 Transition {
  path=AnimGraph.AnimGraphNode_StateMachine_0.Invader Animations.AnimStateTransitionNode_4.Transition
  nodes [Result(bCanEnterTransition=Get Is Jumping())]
 }
 Transition {
  path=AnimGraph.AnimGraphNode_StateMachine_0.Invader Animations.AnimStateTransitionNode_5.Transition
  nodes [Result(bCanEnterTransition=NOT Boolean(A=Get Is Playing Melee Anim()))]
 }
 Transition {
  path=AnimGraph.AnimGraphNode_StateMachine_0.Invader Animations.AnimStateTransitionNode_6.Transition
  nodes [Result(bCanEnterTransition=OR Boolean!(A=Get Is Playing Melee Anim(), B=Get Is DBNO(), C:bool=Get Is Playing Emote()) -> (ReturnValue))]
 }
 Transition {
  path=AnimGraph.AnimGraphNode_StateMachine_0.Invader Animations.AnimStateTransitionNode_7.Transition
  nodes [Result(bCanEnterTransition=Get Is DBNO())]
 }
 Transition {
  path=AnimGraph.AnimGraphNode_StateMachine_0.Invader Animations.AnimStateTransitionNode_8.Transition
  nodes [Result(bCanEnterTransition=NOT Boolean(A=Get Is DBNO()))]
 }
 Transition {path=AnimGraph.AnimGraphNode_StateMachine_0.Invader Animations.AnimStateTransitionNode_9.Transition, nodes [Result()]}
 EventGraph {nodes [
   CrouchStart() {is=title K2Node_CustomEvent}
   > Set SpringStiffness(SpringStiffness:double=1) -> (Output_Get:double)
   > Set SpringCriticalDampingFactor(SpringCriticalDampingFactor:double=0.9) -> (Output_Get:double)
   > RetriggerableDelay(Duration=1)
   > Set SpringStiffness(SpringStiffness=10)
   > Set SpringStiffnessInterpSpeed(SpringStiffnessInterpSpeed:double=5) -> (Output_Get:double)
   > Set SpringCriticalDampingFactor(SpringCriticalDampingFactor=0.5)
   CrouchStop() {is=title K2Node_CustomEvent}
   > Set SpringStiffness(SpringStiffness=1)
   > Set SpringCriticalDampingFactor(SpringCriticalDampingFactor=0.9)
   > RetriggerableDelay(Duration=1)
   > Set SpringStiffness(SpringStiffness=10)
   > Set SpringStiffnessInterpSpeed(SpringStiffnessInterpSpeed=5)
   > Set SpringCriticalDampingFactor(SpringCriticalDampingFactor=0.5)
   n1 = MeleeStart() {is=title K2Node_CustomEvent}
   Set SpringStiffness(execute=[@n1.then, @n12.then], SpringStiffness=50)
   > Set SpringStiffnessInterpSpeed(SpringStiffnessInterpSpeed=8)
   > Set SpringCriticalDampingFactor(SpringCriticalDampingFactor=0.75)
   > Delay(Duration=1)
   > Set SpringStiffnessInterpSpeed(SpringStiffnessInterpSpeed=1)
   n2 = MeleeStop() {is=title K2Node_CustomEvent}
   Set SpringStiffnessInterpSpeed(execute=[@n2.then, @n13.then], SpringStiffnessInterpSpeed=0)
   > Set SpringStiffnessInterpSpeedSmooth(SpringStiffnessInterpSpeedSmooth:double=0) -> (Output_Get:double)
   > Set SpringStiffness(SpringStiffness=10)
   > Set SpringCriticalDampingFactor(SpringCriticalDampingFactor=0.5)
   SetBackPackTargetVectorAlpha1() {is=title K2Node_CustomEvent}
   > Set BackPackTargetVectorAlpha(BackPackTargetVectorAlpha:double=SelectFloat(A=0.3, B=0.6, bPickA=Get Is Playing Melee Anim())) -> (Output_Get:double)
   SetBackPackTargetVectorAlpha0() {is=title K2Node_CustomEvent}
   > Set BackPackTargetVectorAlpha(BackPackTargetVectorAlpha=0)
   CastOwnersSuccess() {is=title K2Node_CustomEvent}
   > n3 = Sequence()
   K2Node_Composite_12 = CastOwnersFrontEnd!(execute=@n10.then, Success:Exec=@n3.then_0) -> (then:Exec) {is=title K2Node_Composite}
   > K2Node_Composite_10 = FrontEndValues!(execute) -> (then:Exec) {is=title K2Node_Composite}
   K2Node_Composite_13 = CastOwnersGameplay!(execute=@n10.else, Success:Exec=@n3.then_1) -> (then:Exec) {is=title K2Node_Composite}
   > K2Node_Composite_14 = SetLocalVariablesGameplay!(execute) -> (then:Exec) {is=title K2Node_Composite}
   > n4 = Sequence!(execute) -> (then_0, then_1, then_2:Exec)
   n5 = Get Fort Player Pawn() -> (Fort Player Pawn:Fort Player Pawn Object Reference)
   > Set IsMovingOnGround(IsMovingOnGround:bool=IsMovingOnGround!(self:Nav Movement Component Object Reference=Get CharacterMovement!(self:Character Object Reference=@n5) -> (CharacterMovement:Character Movement Component Object Reference)) -> (ReturnValue:bool)) -> (Output_Get:bool)
   > Set RootYawOffset_TurnInPlace_Smooth(RootYawOffset_TurnInPlace_Smooth:double=FInterpTo(
     Current=Get RootYawOffset_TurnInPlace_Smooth() -> (RootYawOffset_TurnInPlace_Smooth:double)
     Target=Get Root Yaw Offset() -> (Root Yaw Offset:double)
     DeltaTime=Get Delta Time X() -> (Delta Time X:double)
     InterpSpeed=6
    )) -> (Output_Get:double)
   > K2Node_Composite_11 = TriggerCrouchEventsForSpringChange!(execute) -> (then_1:Exec) {is=title K2Node_Composite}
   > K2Node_Composite_9 = TriggerMeleeEventsForSpringChange!(execute) -> (then_1:Exec) {is=title K2Node_Composite}
   > K2Node_Composite_0 = OpenParachuteOffset!(execute) -> (then_2:Exec) {is=title K2Node_Composite}
   > K2Node_Composite_2 = JumpUpOffset!(execute) -> (then:Exec) {is=title K2Node_Composite}
   > K2Node_Composite_4 = Spring Values!(execute) -> (then:Exec) {is=title K2Node_Composite}
   > K2Node_Composite_5 = Convert Offset Rate of change to rotation value!(execute) -> (then:Exec) {is=title K2Node_Composite}
   > K2Node_Composite_6 = Convert Local Velocity to rotation value!(execute) -> (then:Exec) {is=title K2Node_Composite}
   > K2Node_Composite_7 = Skydive Dive Pitch!(execute) -> (then:Exec) {is=title K2Node_Composite}
   > K2Node_Composite_3 = MovingCarrierAwayFromEmotes!(execute) -> (then_1:Exec) {is=title K2Node_Composite}
   > K2Node_Composite_15 = GetPlayerSpine05Location!(execute) -> (then:Exec) {is=title K2Node_Composite}
   > K2Node_Composite_8 = SetBackPackTargetAlpha!(execute) -> (then:Exec) {is=title K2Node_Composite}
   Set HoverRotationOffsetTotal(
    execute=[@n22.then, @K2Node_Composite_8.then]
    HoverRotationOffsetTotal:Rotator=Combine Rotators!(
     A:Rotator=Get CarrierOffsetRotationReactionSmooth() -> (CarrierOffsetRotationReactionSmooth:Rotator)
     B:Rotator=Get SkydiveDivingRotationOffset() -> (SkydiveDivingRotationOffset:Rotator)
    ) -> (ReturnValue:Rotator)
   ) -> (Output_Get:Rotator)
   n6 = Branch(execute=@n4.then_1, Condition=Get Is DBNOCarried() -> (Is DBNOCarried:bool))
   > n7 = Do Once(Reset=@n6.else)
   n8 = Set IsMindContollingPlayer(execute=[@n19.then, @n7.Completed], IsMindContollingPlayer:bool=false) -> (Output_Get:bool)
   IsMindControllingPlayer!(
    execute=[@n18.else, @n8.then]
    self:B Pet Inst Invader Object Reference=Get As B Pet Inst Invader() -> (As B Pet Inst Invader:B Pet Inst Invader Object Reference)
    bIsMindcontrolling:bool=Get IsMindContollingPlayer() -> (IsMindContollingPlayer:bool)
   ) -> (then)
   K2Node_Composite_1 = Values Last Tick(execute=@n4.then_2) {is=title K2Node_Composite}
   n9 = Event BlueprintUpdateAnimation()
   > Set Delta Time X(Delta Time X:double=@n9.DeltaTimeX) -> (Output_Get:double)
   > n10 = Branch(Condition=Get bIsInFrontEnd())
   AnimNotify_RandomBreaks_Entered() {is=title K2Node_Event}
   > Set Allow Locomotion Break(Allow Locomotion Break:bool=false) -> (Output_Get:bool)
   n11 = AnimNotify_RandomBreaks_Exited() {is=title K2Node_Event}
   Set SpringStiffness(execute=[@n11.then, @n16.then], SpringStiffness=10)
   > Set SpringStiffnessInterpSpeed(SpringStiffnessInterpSpeed=5)
   > Set SpringCriticalDampingFactor(SpringCriticalDampingFactor=0.5)
   > Set IsPlayingBreakAnim(IsPlayingBreakAnim:bool=false) -> (Output_Get:bool)
   n12 = AnimNotify_Hunker_Entered() {is=title K2Node_Event}
   n13 = AnimNotify_Hunker_Exited() {is=title K2Node_Event}
   AnimNotify_SetSparkOn() {is=title K2Node_Event}
   > n14 = Set SparkOn(SparkOn:bool=true) -> (Output_Get:bool)
   SetSParkVFX!(
    execute=[@n15.then, @n14.then]
    self:B Pet Inst Invader Object Reference=Get As B Pet Inst Invader()
    SparkOnParam:bool=Get SparkOn() -> (SparkOn:bool)
   ) -> (then)
   AnimNotify_SetSparkOff() {is=title K2Node_Event}
   > n15 = Set SparkOn(SparkOn=false)
   n16 = AnimNotify_Locomotion_FullyBlended() {is=title K2Node_Event}
   AnimNotify_DBNO_Entered() {is=title K2Node_Event}
   > Set SpringStiffness(SpringStiffness=0.5)
   > Set SpringStiffnessInterpSpeed(SpringStiffnessInterpSpeed=0.5)
   AnimNotify_TriggerBreakRandomizer() {is=title K2Node_Event}
   > n17 = Set Allow Locomotion Break(Allow Locomotion Break=RandomBoolWithWeight!(Weight:float=0.1) -> (ReturnValue:bool))
   > Branch(Condition=@n17)
   > Set Choose Loco Break Index(Choose Loco Break Index:int=RandomInteger!(Max:int=2) -> (ReturnValue:int=0)) -> (Output_Get:int=0)
   > Set IsPlayingBreakAnim(IsPlayingBreakAnim=true)
   AnimNotify_IsMindControllingPlayer() {is=title K2Node_Event}
   > Set IsMindContollingPlayer(IsMindContollingPlayer=true)
   > n18 = Branch(Condition=Get Is DBNOCarried())
   n19 = AnimNotify_IsNotMindcontrollingPlayer() {is=title K2Node_Event}
   AnimNotify_CancelBreakOffset() {is=title K2Node_Event}
   > Set IsPlayingBreakAnim(IsPlayingBreakAnim=false)
   > Set SpringStiffness(SpringStiffness=1)
   > Set SpringCriticalDampingFactor(SpringCriticalDampingFactor=0.9)
   n20 = Get RootYawOffset_TurnInPlace_Smooth()
   n21 = Get RootRotationMeleeTwistSpring() -> (RootRotationMeleeTwistSpring:double)
   n22 = Set RootRotationOffsetTotal(RootRotationOffsetTotal:Rotator=Combine Rotators(A=MakeRotator(Yaw=@n21), B=MakeRotator(Yaw=@n20))) -> (Output_Get:Rotator)
   PrintString!(execute, InString=IDLE EXITED, bPrintToScreen, bPrintToLog, TextColor, Duration, Key) -> (then)
   PrintString(InString=IDLE ENTERED)
   IsMindControllingPlayer(self=Get As B Pet Inst Invader(), bIsMindcontrolling=Get IsMindContollingPlayer())
   Event BlueprintInitializeAnimation()
   AnimNotify_Idle_Exited() {is=title K2Node_Event}
   AnimNotify_Idle_Entered() {is=title K2Node_Event}
   AnimNotify_DBNO_Exited() {is=title K2Node_Event}
  ]}
 OpenParachuteOffset {
  path=EventGraph.K2Node_Composite_0.OpenParachuteOffset
  nodes [
   Inputs()
   > n1 = Sequence()
   > Branch(Condition=Get Is Parachute Open() -> (Is Parachute Open:bool))
   > n2 = Branch(Condition=Get Is Parachute Open Last Tick() -> (Is Parachute Open Last Tick:bool))
   Do Once(execute=@n2.else, Reset=@n2.then)
   > Set Is Parachute Just Opened(Is Parachute Just Opened:bool=true) -> (Output_Get:bool)
   > RetriggerableDelay(Duration=1.2)
   > Set Is Parachute Just Opened(Is Parachute Just Opened=false)
   Outputs!(then_2:Exec=@n1.then_2)
   n3 = Get Is Parachute Just Opened() -> (Is Parachute Just Opened:bool)
   Set Parachute Just Opened Offset(
    execute=@n1.then_1
    Parachute Just Opened Offset:Vector=VInterpTo!(
     Current:Vector=Get Parachute Just Opened Offset() -> (Parachute Just Opened Offset:Vector)
     Target:Vector=SelectVector(A="0,20,30", B="0, 0, 0", bPickA=@n3)
     DeltaTime:float=Get Delta Time X()
     InterpSpeed:float=SelectFloat(A=3, B=1, bPickA=@n3)
    ) -> (ReturnValue:Vector)
   ) -> (Output_Get:Vector)
  ]
 }
 Values Last Tick {
  path=EventGraph.K2Node_Composite_1.Values Last Tick
  nodes [
   Inputs()
   > Set Is Parachute Open Last Tick(Is Parachute Open Last Tick:bool=Get Is Parachute Open()) -> (Output_Get:bool)
   > Set Carrier Offset Total Last Tick(Carrier Offset Total Last Tick:Vector=Get CarrierOffsetTotal()) -> (Output_Get:Vector)
   Outputs!()
  ]
 }
 FrontEndValues {
  path=EventGraph.K2Node_Composite_10.FrontEndValues
  nodes [
   Inputs()
   > n1 = Sequence!(execute) -> (then_0, then_1, then_2, then_3:Exec, then_4:Exec)
   n2 = Get Mesh(self=Get Fort Player Pawn())
   n3 = GetSocketTransform(self=Get Mesh(self=Get Fort Player Pawn()), InSocketName=spine_05, TransformSpace=RTS_World)
   n4 = MakeVector(X=15, Y=-40)
   > n5 = Set BackPackTargetVector(BackPackTargetVector:Vector=vector + vector(
     A:Vector=GetSocketLocation!(self:Scene Component Object Reference=@n2, InSocketName:Name=spine_05) -> (ReturnValue:Vector)
     B:Vector=TransformDirection!(T:"Transform (by ref)"=@n3, Direction:Vector=@n4) -> (ReturnValue:Vector)
    ) -> (ReturnValue:Vector) {is=title K2Node_CommutativeAssociativeBinaryOperator}) -> (Output_Get:Vector)
   n6 = Get FrontEndAnimBP() -> (FrontEndAnimBP:Frontend Anim Instance Object Reference)
   Set Skydiving Is Skydiving(execute=@n1.then_1, Skydiving Is Skydiving:bool=Get bIsSkydiving!(self:Frontend Anim Instance Object Reference=@n6) -> (bIsSkydiving:bool)) -> (Output_Get:bool)
   > Set Skydiving Is Skydive Dive Mode(Skydiving Is Skydive Dive Mode:bool=Get bIsSkydiveDiving!(self:Frontend Anim Instance Object Reference=@n6) -> (bIsSkydiveDiving:bool)) -> (Output_Get:bool)
   n7 = Get Skydiving Is Skydiving()
   n8 = Get FrontEndAnimBP()
   > Set Is Playing Emote(Is Playing Emote:bool=AND Boolean(A=Get bIsPlayingEmote!(self:Frontend Anim Instance Object Reference=@n8) -> (bIsPlayingEmote), B=NOT Boolean(A=@n7))) -> (Output_Get:bool)
   > n9 = Branch(Condition=Get Skydiving Is Skydiving())
   > Set YawDefinedHoverOffset(YawDefinedHoverOffset:double=0) -> (Output_Get:double)
   > Set SpringStiffnessInterpSpeed(SpringStiffnessInterpSpeed=100)
   > n10 = Branch(Condition=Get Skydiving Is Skydive Dive Mode())
   > Set BaseAimPitchSmooth(BaseAimPitchSmooth:double=30) -> (Output_Get:double)
   Set BaseAimPitchSmooth(execute=@n10.else, BaseAimPitchSmooth=10)
   Set BaseAimPitchSmooth(execute=@n9.else, BaseAimPitchSmooth=0)
   > Set SpringStiffnessInterpSpeed(SpringStiffnessInterpSpeed=2)
   > n11 = Branch(Condition=Get bIsPlayingEmote(self=Get FrontEndAnimBP()))
   > Set YawDefinedHoverOffset(YawDefinedHoverOffset=0)
   n12 = GetSocketTransform(self=GetSkeletalMeshForPartType(self=Get Fort Player Pawn()), InSocketName=Backpack_BR, TransformSpace=RTS_Component)
   Set YawDefinedHoverOffset(execute=@n11.else, YawDefinedHoverOffset=BreakRotator(InRot=BreakTransform(InTransform=@n12).Rotation).Yaw)
   n13 = SelectFloat(A=0, B=1, bPickA=Get Is Playing Emote())
   n14 = Get Skydiving Is Skydiving()
   Set BackPackTargetVectorAlphaSmooth(
    execute=@n1.then_3
    BackPackTargetVectorAlphaSmooth:double=FInterpTo(
     Current=Get BackPackTargetVectorAlphaSmooth()
     Target=SelectFloat(A=0, B=@n13, bPickA=@n14)
     DeltaTime=Get Delta Time X()
     InterpSpeed=Get SpringStiffnessInterpSpeedSmooth() -> (SpringStiffnessInterpSpeedSmooth:double)
    )
   ) -> (Output_Get:double)
   n15 = MakeRotator(Roll=Get BaseAimPitchSmooth() -> (BaseAimPitchSmooth:double), Yaw=Get YawDefinedHoverOffset() -> (YawDefinedHoverOffset:double))
   n16 = Get Is Playing Emote()
   > Set RotationYawDefinedHoverOffset(RotationYawDefinedHoverOffset:Rotator=RInterpTo!(
     Current:Rotator=Get RotationYawDefinedHoverOffset()
     Target:Rotator=SelectRotator!(A:Rotator="0, 0, 0", B:Rotator=@n15, bPickA:bool=@n16) -> (ReturnValue:Rotator)
     DeltaTime:float=Get Delta Time X()
     InterpSpeed:float=3
    ) -> (ReturnValue:Rotator)) -> (Output_Get:Rotator)
   n17 = SelectVector(A="0,-20,65", B="60,-40,32", bPickA=Get Is Playing Emote())
   n18 = Get Skydiving Is Skydive Dive Mode()
   > Set CarrierOffsetTotal(CarrierOffsetTotal:Vector=VInterpTo(Current=Get CarrierOffsetTotal(), Target=SelectVector(A="0,-20,65", B=@n17, bPickA=@n18), DeltaTime=Get Delta Time X(), InterpSpeed=2)) -> (Output_Get:Vector)
   Outputs!(then:Exec=@n1.then_4)
   Set SpringStiffnessInterpSpeedSmooth(
    execute=@n1.then_2
    SpringStiffnessInterpSpeedSmooth=FInterpTo(
     Current=Get SpringStiffnessInterpSpeedSmooth()
     Target=Get SpringStiffnessInterpSpeed() -> (SpringStiffnessInterpSpeed:double)
     DeltaTime=Get Delta Time X()
     InterpSpeed=2
    )
   )
   DrawDebugSphere!(
    execute
    Center:Vector=@n5
    Radius:float=10
    Segments:int=12
    LineColor:Linear Color Structure=(R=1,G=1,B=1,A=1)
    Duration:float=0
    Thickness:float=0.5
    DepthPriority:EDrawDebugSceneDepthPriorityGroup Enum=World
   ) -> (then)
  ]
 }
 TriggerCrouchEventsForSpringChange {
  path=EventGraph.K2Node_Composite_11.TriggerCrouchEventsForSpringChange
  nodes [
   Inputs()
   > n1 = Sequence!(execute) -> (then_0, then_1)
   > n2 = Branch(Condition=Get Is Crouching() -> (Is Crouching:bool))
   > n3 = Sequence()
   > Do Once(Reset=@n4.then_1)
   > CrouchStart() {is=title K2Node_CallFunction}
   Do Once(execute=@n4.then_0, Reset=@n3.then_1)
   > CrouchStop() {is=title K2Node_CallFunction}
   n4 = Sequence(execute=@n2.else)
   Outputs!(then_1:Exec=@n1.then_1)
  ]
 }
 CastOwnersFrontEnd {
  path=EventGraph.K2Node_Composite_12.CastOwnersFrontEnd
  nodes [
   n1 = Inputs!() -> (execute, Success:Exec)
   > n2 = Sequence()
   > Gate!(Enter:Exec, Open:Exec, Close:Exec=@n11.then_0, Toggle:Exec, bStartClosed:bool=false) -> (Exit:Exec) {is=title K2Node_MacroInstance}
   n3 = GetOwningActor()
   > Is Valid(InputObject=@n3)
   > n4 = Cast To B_PetInst_Invader!(execute, Object=@n3) -> (then, CastFailed, AsB Pet Inst Invader:B Pet Inst Invader Object Reference) {is=title K2Node_DynamicCast}
   > n5 = Set As B Pet Inst Invader(As B Pet Inst Invader:B Pet Inst Invader Object Reference=@n4) -> (Output_Get:B Pet Inst Invader Object Reference)
   n6 = GetOwner(self:Actor Object Reference=@n5)
   > Is Valid(InputObject=@n6)
   > n7 = Cast To FortPlayerPawn(Object=@n6)
   > Set Fort Player Pawn(Fort Player Pawn:Fort Player Pawn Object Reference=@n7) -> (Output_Get:Fort Player Pawn Object Reference)
   n8 = GetOwnerFrontEndAnimInstance()
   > Is Valid(InputObject=@n8)
   > Set FrontEndAnimBP(FrontEndAnimBP:Frontend Anim Instance Object Reference=@n8) -> (Output_Get:Frontend Anim Instance Object Reference)
   > n9 = CastOwnersSuccess() {is=title K2Node_CallFunction}
   Outputs!(then:Exec=[@n9.then, @n10.Exit])
   n10 = Gate(Enter=@n2.then_1, Open=@n11.then_1, bStartClosed=true)
   n11 = Sequence(execute=@n1.Success)
  ]
 }
 CastOwnersGameplay {
  path=EventGraph.K2Node_Composite_13.CastOwnersGameplay
  nodes [
   n1 = Inputs()
   > n2 = Sequence()
   > Gate(Close=@n12.then_0, bStartClosed=false)
   n3 = GetOwningActor()
   > Is Valid(InputObject=@n3)
   > n4 = Cast To B_PetInst_Invader(Object=@n3)
   > n5 = Set As B Pet Inst Invader(As B Pet Inst Invader=@n4)
   n6 = GetOwner(self=@n5)
   > Is Valid(InputObject=@n6)
   > n7 = Cast To FortPlayerPawn(Object=@n6)
   > Set Fort Player Pawn(Fort Player Pawn=@n7)
   n8 = GetAnimInstance(self=Get Mesh(self=Get Fort Player Pawn()))
   > Is Valid(InputObject=@n8)
   > n9 = Cast To Fortnite_M_Avg_Player_AnimBlueprint!(execute, Object=@n8) -> (then, CastFailed, AsFortnite M Avg Player Anim Blueprint:Fortnite M Avg Player Anim Blueprint Object Reference) {is=title K2Node_DynamicCast}
   > Set Fort Player AnimBP(Fort Player AnimBP:Fortnite M Avg Player Anim Blueprint Object Reference=@n9) -> (Output_Get:Fortnite M Avg Player Anim Blueprint Object Reference)
   > n10 = CastOwnersSuccess()
   Outputs(then=[@n10.then, @n11.Exit])
   n11 = Gate(Enter=@n2.then_1, Open=@n12.then_1, bStartClosed=true)
   n12 = Sequence(execute=@n1.Success)
  ]
 }
 SetLocalVariablesGameplay {
  path=EventGraph.K2Node_Composite_14.SetLocalVariablesGameplay
  nodes [
   Inputs!() -> (execute)
   n1 = Get Fort Player AnimBP() -> (Fort Player AnimBP:Fortnite M Avg Player Anim Blueprint Object Reference)
   n2 = Get!(self:Fort Player Anim Instance Object Reference=@n1) -> (LocalVelocity_X:double, LocalVelocity_Y:double, LocalVelocity_Z:double)
   > Set LocalVelocity(LocalVelocity:Vector=MakeVector(X=@n2, Y=@n2.LocalVelocity_Y, Z=Get VelocityZ!(self:Fort Player Anim Instance Object Reference=@n1) -> (VelocityZ:float))) -> (Output_Get:Vector)
   n3 = Break Fort Anim Input Skydiving!(FortAnimInput_Skydiving:"Fort Anim Input Skydiving Structure (by ref)"=Get Skydiving!(self:Fort Player Anim Instance Object Reference=Get Fort Player AnimBP()) -> (Skydiving:Fort Anim Input Skydiving Structure)) -> (
    ParaGilderRootOffset:Vector=~
    ParaGliderRootRotationOverride:Rotator=~
    bUseParaGlideRootModifier:bool=~
    bIsSkydivingFromLaunchPad:bool=~
    bIsSkydivingFromBus:bool=~
    bIsInVortex:bool=~
    bIsDiving:bool=~
    bIsParachuteOpen:bool=~
    bIsSkydiving:bool=~
    bIsParachuteLeaning:bool=~
    bIsSkydiveLeaning:bool=~
    bIsLeaning:bool=~
    bIsSkydiveDiveMode:bool=~
    bPlaySkydiveDrift:bool=~
    bSkydiveDriftDelayActive:bool=~
    bSkydiveDriftAnimAllowed:bool=~
    bIsGliderRight:bool=~
    bIsGliderCenter:bool=~
    bIsGliderLeft:bool=~
    bIsGliderForward:bool=~
    bIsGliderBack:bool=~
    bHasDeployedAndIsStandingUpright:bool=~
    LocalAccelForward:float=~
    LocalAccelRight:float=~
    LocalVelocityRight:float=~
    SkydiveAimPitch:float=~
    SkydiveAimYaw:float=~
    DeployChuteAnimRate:float=~
    SkydiveDriftAnimRate:float=~
    SkydiveDriftAnimRateCurrent:float=~
    SkydiveFidgetAnimRate:float=~
    SkydiveFidgetAnimRateCurrent:float=~
    SkydiveAdditiveAlpha:float=~
    SkydiveDriftDelay:float=~
    SkydiveDriftAnim:int
    SkydiveDriftAnimMax:int
    LaunchpadAnim:int
    LocalAccelDir:ESkydivingDirection Enum
    DirectionLast:ESkydivingDirection Enum
    PoseAdditive:Anim Sequence Object Reference
   ) {is=title K2Node_BreakStruct}
   > Set Is Skydiving(Is Skydiving:bool=@n3.bIsSkydiving) -> (Output_Get:bool)
   > Set Is Parachute Open(Is Parachute Open:bool=@n3.bIsParachuteOpen) -> (Output_Get:bool)
   > Set Is Diving(Is Diving:bool=@n3.bIsDiving) -> (Output_Get:bool)
   n4 = GetTurnInPlaceInput!(self:Fort Player Animation Interface Object Reference=GetOwnerAnimationInterface!(self:Custom Character Part Anim Instance Object Reference) -> (ReturnValue:Fort Player Animation Interface Object Reference)) -> (ReturnValue:"Fort Anim Input Turn In Place Structure (by ref)")
   > Set Root Yaw Offset(Root Yaw Offset:double=float * float(
     A=Break Fort Anim Input Turn in Place!(FortAnimInput_TurnInPlace:"Fort Anim Input Turn In Place Structure (by ref)"=@n4) -> (
      RootYawOffset:float=~
      TurnAnimPlayRate:float=~
      bWantsToTurn:bool=~
      bExitTurn:bool=~
      bUseTurnA:bool=~
      TurnACurveAlpha:float=~
      TurnBCurveAlpha:float=~
      bWantsToTurnInVehicle:bool=~
      bWantsToTurnAgain:bool=~
      bTurningLeft:bool=~
      ShufflePlayRate:float=~
      CrouchShufflePlayRate:float=~
      CrouchShuffleStartPosition:float=~
      CrouchShuffleSpeedScaling:float=~
      bTransition_LowerBody_Shuffle:bool=~
      bTransition_LowerBody_ShuffleADS:bool=~
      bPlayOutro:bool=~
     ) {is=title K2Node_BreakStruct}
     B=0.6
    )) -> (Output_Get:double)
   n5 = Get Fort Player AnimBP()
   > Set Is Playing Melee Anim(Is Playing Melee Anim:bool=Get bIsPlayingMeleeAnim!(self:Fort Player Anim Instance Object Reference=@n5) -> (bIsPlayingMeleeAnim)) -> (Output_Get:bool)
   > Set Transition Locomotion Additive Crouch Turning(Transition Locomotion Additive Crouch Turning:bool=Get bTransition_LocomotionAdditive_CrouchTurning!(self:Fort Player Anim Instance Object Reference=@n5) -> (bTransition_LocomotionAdditive_CrouchTurning:bool)) -> (Output_Get:bool)
   > Set Is Moving 2D(Is Moving 2D:bool=Get bIsMoving2D!(self:Fort Player Anim Instance Object Reference=@n5) -> (bIsMoving2D)) -> (Output_Get:bool)
   > Set Is Targeting(Is Targeting:bool=Get bIsTargeting!(self:Fort Player Anim Instance Object Reference=@n5) -> (bIsTargeting)) -> (Output_Get:bool)
   > Set Is Crouching(Is Crouching:bool=Get bIsCrouching!(self:Fort Player Anim Instance Object Reference=@n5) -> (bIsCrouching)) -> (Output_Get:bool)
   > Set Is DBNO(Is DBNO:bool=Get bIsDBNO!(self:Fort Player Anim Instance Object Reference=@n5) -> (bIsDBNO)) -> (Output_Get:bool)
   > Set Is Playing Emote(Is Playing Emote=Get bIsPlayingEmote(self:Fort Player Anim Instance Object Reference=@n5))
   > Set Is Sprinting(Is Sprinting:bool=Get bIsSprinting!(self:Fort Player Anim Instance Object Reference=@n5) -> (bIsSprinting)) -> (Output_Get:bool)
   > Set Is Surface Swimming(Is Surface Swimming:bool=Get bIsSurfaceSwimming!(self:Fort Player Anim Instance Object Reference=@n5) -> (bIsSurfaceSwimming)) -> (Output_Get:bool)
   > Set Is Dive Jumping(Is Dive Jumping:bool=Get bIsDiveJumping!(self:Fort Player Anim Instance Object Reference=@n5) -> (bIsDiveJumping:bool)) -> (Output_Get:bool)
   > Set Is Jumping(Is Jumping:bool=Get bIsJumping!(self:Fort Anim Instance Object Reference=@n5) -> (bIsJumping)) -> (Output_Get:bool)
   > Set Velocity Z(Velocity Z:double=Get VelocityZ(self=@n5)) -> (Output_Get:double)
   > Set Melee Twist Curve Value(Melee Twist Curve Value:double=Get MeleeTwistCurveValue!(self:Fort Player Anim Instance Object Reference=@n5) -> (MeleeTwistCurveValue:float)) -> (Output_Get:double)
   > Set Yaw Delta Current Tick(Yaw Delta Current Tick:double=Get YawDeltaCurrentTick!(self:Fort Player Anim Instance Object Reference=@n5) -> (YawDeltaCurrentTick:float)) -> (Output_Get:double)
   > Set Is Relaxed Level 1(Is Relaxed Level 1:bool=Get bIsRelaxedLevel1!(self:Fort Anim Instance Object Reference=@n5) -> (bIsRelaxedLevel1:bool)) -> (Output_Get:bool)
   > Set Recently Fired(Recently Fired:bool=Get bRecentlyFired!(self:Fort Anim Instance Object Reference=@n5) -> (bRecentlyFired:bool)) -> (Output_Get:bool)
   > Set Local Velocity Yaw Angle(Local Velocity Yaw Angle:double=Get LocalVelocityYawAngle!(self:Fort Player Anim Instance Object Reference=@n5) -> (LocalVelocityYawAngle:float)) -> (Output_Get:double)
   n6 = Get!(self=Get Fort Player AnimBP()) -> (
    Skydiving_ParaGilderRootOffset:Vector=~
    Skydiving_ParaGliderRootRotationOverride:Rotator=~
    Skydiving_bUseParaGlideRootModifier:bool=~
    Skydiving_bIsSkydivingFromLaunchPad:bool=~
    Skydiving_bIsSkydivingFromBus:bool=~
    Skydiving_bIsInVortex:bool=~
    Skydiving_bIsDiving:bool=~
    Skydiving_bIsParachuteOpen:bool=~
    Skydiving_bIsSkydiving:bool=~
    Skydiving_bIsParachuteLeaning:bool=~
    Skydiving_bIsSkydiveLeaning:bool=~
    Skydiving_bIsLeaning:bool=~
    Skydiving_bIsSkydiveDiveMode:bool=~
    Skydiving_bPlaySkydiveDrift:bool=~
    Skydiving_bSkydiveDriftDelayActive:bool=~
    Skydiving_bSkydiveDriftAnimAllowed:bool=~
    Skydiving_bIsGliderRight:bool=~
    Skydiving_bIsGliderCenter:bool=~
    Skydiving_bIsGliderLeft:bool=~
    Skydiving_bIsGliderForward:bool=~
    Skydiving_bIsGliderBack:bool=~
    Skydiving_bHasDeployedAndIsStandingUpright:bool=~
    Skydiving_LocalAccelForward:float=~
    Skydiving_LocalAccelRight:float=~
    Skydiving_LocalVelocityRight:float=~
    Skydiving_SkydiveAimPitch:float=~
    Skydiving_SkydiveAimYaw:float=~
    Skydiving_DeployChuteAnimRate:float=~
    Skydiving_SkydiveDriftAnimRate:float=~
    Skydiving_SkydiveDriftAnimRateCurrent:float=~
    Skydiving_SkydiveFidgetAnimRate:float=~
    Skydiving_SkydiveFidgetAnimRateCurrent:float=~
    Skydiving_SkydiveAdditiveAlpha:float=~
    Skydiving_SkydiveDriftDelay:float=~
    Skydiving_SkydiveDriftAnim:int
    Skydiving_SkydiveDriftAnimMax:int
    Skydiving_LaunchpadAnim:int
    Skydiving_LocalAccelDir:ESkydivingDirection Enum
    Skydiving_DirectionLast:ESkydivingDirection Enum
    Skydiving_PoseAdditive:Anim Sequence Object Reference
   )
   > Set Skydiving Is Skydiving(Skydiving Is Skydiving=@n6.Skydiving_bIsSkydiving)
   > Set Skydiving Is Diving(Skydiving Is Diving:bool=@n6.Skydiving_bIsDiving) -> (Output_Get:bool)
   > Set Skydiving Is Skydive Dive Mode(Skydiving Is Skydive Dive Mode=@n6.Skydiving_bIsSkydiveDiveMode)
   > Set Skydiving Is Parachute Open(Skydiving Is Parachute Open:bool=@n6.Skydiving_bIsParachuteOpen) -> (Output_Get:bool)
   > n7 = Set Is DBNOCarried(Is DBNOCarried:bool=Get bIsDBNOCarried!(self:Fort Player Anim Instance Object Reference=Get Fort Player AnimBP()) -> (bIsDBNOCarried)) -> (Output_Get:bool)
   > Set IsDBNOCarriedAlpha(IsDBNOCarriedAlpha:double="To Float (Boolean)"(InBool=@n7)) -> (Output_Get:double)
   > Outputs()
   Get!() -> (PawnVelocity_X:double, PawnVelocity_Y:double, PawnVelocity_Z:double)
  ]
 }
 GetPlayerSpine05Location {
  path=EventGraph.K2Node_Composite_15.GetPlayerSpine05Location
  nodes [
   Inputs()
   > n1 = Sequence!(execute) -> (then_0, then_1, then_2:Exec)
   n2 = Get Mesh(self=Get Fort Player Pawn())
   n3 = GetSocketTransform(self=Get Mesh(self=Get Fort Player Pawn()), InSocketName=spine_05, TransformSpace=RTS_World)
   n4 = MakeVector(X=5, Y=-50)
   > n5 = Set BackPackTargetVector(BackPackTargetVector=vector + vector(A=GetSocketLocation(self=@n2, InSocketName=spine_05), B=TransformDirection(T=@n3, Direction=@n4)))
   Set BackPackTargetVectorAlphaSmooth(
    execute=@n1.then_1
    BackPackTargetVectorAlphaSmooth=FInterpTo(
     Current=Get BackPackTargetVectorAlphaSmooth()
     Target=Get BackPackTargetVectorAlpha() -> (BackPackTargetVectorAlpha:double)
     DeltaTime=Get Delta Time X()
     InterpSpeed=4
    )
   )
   Outputs(then=@n1.then_2)
   DrawDebugSphere(Center=@n5, Radius=10, Segments=12, LineColor=(R=1,G=1,B=1,A=1), Duration=0, Thickness=0.5, DepthPriority=World)
  ]
 }
 JumpUpOffset {
  path=EventGraph.K2Node_Composite_2.JumpUpOffset
  nodes [
   Inputs()
   n1 = AND Boolean(A=Get Is Jumping(), B="float > float"(A=Get Velocity Z() -> (Velocity Z:double), B=0) {is=title K2Node_CallFunction})
   > Set Jump Up Offset Smooth(Jump Up Offset Smooth:Vector=VInterpTo(
     Current=Get Jump Up Offset Smooth() -> (Jump Up Offset Smooth:Vector)
     Target=SelectVector(A="0,-10,25", B="0, 0, 0", bPickA=@n1)
     DeltaTime=Get Delta Time X()
     InterpSpeed=10
    )) -> (Output_Get:Vector)
   > Outputs()
  ]
 }
 MovingCarrierAwayFromEmotes {
  path=EventGraph.K2Node_Composite_3.MovingCarrierAwayFromEmotes
  nodes [
   Inputs()
   > n1 = Sequence!(execute) -> (then_0, then_1)
   > n2 = Branch(Condition=Get Is Playing Emote())
   n3 = Get Fort Player Pawn()
   > n4 = Branch(Condition=IsNullItem!(ItemObject:Object Reference=GetCurrentMontage!(self:Character Object Reference=@n3) -> (ReturnValue:Anim Montage Object Reference)) -> (Is Null Item:bool))
   Set Emote_NonPlayerTrigger Montage Playing(execute=[@n4.then, @n2.else], Emote_NonPlayerTrigger Montage Playing:bool=false) -> (Output_Get:bool)
   > Set AnyEmotePlaying(AnyEmotePlaying:bool=false) -> (Output_Get:bool)
   Set AnyEmotePlaying(execute=@n4.else, AnyEmotePlaying=true)
   n5 = Get Fort Player Pawn()
   > Branch(Condition="Equal (Object)"(A=GetCurrentMontage(self=@n5), B=Get Emote_NonPlayerTrigger Montage() -> (Emote_NonPlayerTrigger Montage:Anim Montage Object Reference)) {is=title K2Node_CallFunction})
   > Set Emote_NonPlayerTrigger Montage Playing(Emote_NonPlayerTrigger Montage Playing=true)
   Outputs!(then_1:Exec=@n1.then_1)
  ]
 }
 Spring Values {
  path=EventGraph.K2Node_Composite_4.Spring Values
  nodes [
   Inputs()
   > Set SpringStiffnessInterpSpeedSmooth(SpringStiffnessInterpSpeedSmooth=FInterpTo(Current=Get SpringStiffnessInterpSpeedSmooth(), Target=Get SpringStiffnessInterpSpeed(), DeltaTime=Get Delta Time X(), InterpSpeed=3))
   > Set SpringStiffnessSmooth(SpringStiffnessSmooth:double=FInterpTo(
     Current=Get SpringStiffnessSmooth() -> (SpringStiffnessSmooth:double)
     Target=Get SpringStiffness() -> (SpringStiffness:double)
     DeltaTime=Get Delta Time X()
     InterpSpeed=Get SpringStiffnessInterpSpeedSmooth()
    )) -> (Output_Get:double)
   > Set SpringCriticalDampingFactorSmooth(SpringCriticalDampingFactorSmooth:double=FInterpTo(
     Current=Get SpringCriticalDampingFactorSmooth() -> (SpringCriticalDampingFactorSmooth:double)
     Target=Get SpringCriticalDampingFactor() -> (SpringCriticalDampingFactor:double)
     DeltaTime=Get Delta Time X()
     InterpSpeed=Get SpringStiffnessInterpSpeedSmooth()
    )) -> (Output_Get:double)
   n1 = float * float(A=Get Melee Twist Curve Value() -> (Melee Twist Curve Value:double), B=0.5)
   n2 = Get!() -> (LocalVelocity_X:double, LocalVelocity_Y:double, LocalVelocity_Z:double)
   n3 = "Clamp (Float)"(Value=float * float(A=@n2.LocalVelocity_Y, B=0.03), Min=-30, Max=30)
   n4 = "Clamp (Float)"(Value=float * float(A=@n2.LocalVelocity_Z, B=-0.04), Min=-30, Max=30)
   n5 = "Clamp (Float)"(Value=float * float(A=@n2, B=-0.03), Min=-30, Max=2)
   > n6 = VectorSpringInterp!(
    execute
    Current:Vector=Get CarrierOffsetTotal()
    Target:Vector=vector + vector!(
     A=Get Offset Values!() -> (ReturnValue:Vector=~) {is=title K2Node_MacroInstance}
     B=MakeVector(X=@n3, Y=@n5, Z=@n4)
     C:Vector=MakeVector(X=@n1, Y=0)
     D:Vector=Get Parachute Just Opened Offset()
    ) -> (ReturnValue)
    SpringState:"Vector Spring State Structure (by ref)"=Get Carrier Offset Spring State() -> (Carrier Offset Spring State:Vector Spring State Structure)
    Stiffness:float=Get SpringStiffnessSmooth()
    CriticalDampingFactor:float=Get SpringCriticalDampingFactorSmooth()
    DeltaTime:float=Get Delta Time X()
    Mass:float=0.75
    TargetVelocityAmount:float=1
    bClamp:bool=false
    MinValue:Vector="-1,-1,-1"
    MaxValue:Vector="1,1,1"
    bInitializeFromTarget:bool=false
   ) -> (then, ReturnValue:Vector)
   > Set CarrierOffsetTotal(CarrierOffsetTotal=@n6)
   > n7 = FloatSpringInterp!(
    execute
    Current:float=Get RootRotationMeleeTwistSpring()
    Target:float=float * float(A=Get Melee Twist Curve Value(), B=0.5)
    SpringState:"Float Spring State Structure (by ref)"=Get Carrier Root Rotation Spring State() -> (Carrier Root Rotation Spring State:Float Spring State Structure)
    Stiffness:float=Get SpringStiffnessSmooth()
    CriticalDampingFactor:float=Get SpringCriticalDampingFactorSmooth()
    DeltaTime:float=Get Delta Time X()
    Mass:float=0.75
    TargetVelocityAmount:float=1
    bClamp:bool=false
    MinValue:float=-1
    MaxValue:float=1
    bInitializeFromTarget:bool=false
   ) -> (then, ReturnValue:float)
   > Set RootRotationMeleeTwistSpring(RootRotationMeleeTwistSpring:double=@n7) -> (Output_Get:double)
   > n8 = FloatSpringInterp(
    Current=Get YawDefinedHoverOffset()
    Target=Get Rotation Offset Values!() -> (ReturnValue:double=~) {is=title K2Node_MacroInstance}
    SpringState=Get Carrier Rotation Spring State() -> (Carrier Rotation Spring State:Float Spring State Structure)
    Stiffness=Get SpringStiffnessSmooth()
    CriticalDampingFactor=Get SpringCriticalDampingFactorSmooth()
    DeltaTime=Get Delta Time X()
    Mass=0.75
    TargetVelocityAmount=1
    bClamp=false
    MinValue=-1
    MaxValue=1
    bInitializeFromTarget=false
   )
   > n9 = Set YawDefinedHoverOffset(YawDefinedHoverOffset=@n8)
   > Set RotationYawDefinedHoverOffset(RotationYawDefinedHoverOffset=MakeRotator(Yaw=@n9))
   > Outputs!(then:Exec)
   Get BackPackTargetVector()
  ]
 }
 Convert Offset Rate of change to rotation value {
  path=EventGraph.K2Node_Composite_5.Convert Offset Rate of change to rotation value
  nodes [
   Inputs()
   n1 = Get CarrierOffsetTotal()
   n2 = Get Carrier Offset Total Last Tick() -> (Carrier Offset Total Last Tick:Vector)
   > n3 = Set CarrierPackOffsetRate(CarrierPackOffsetRate:Vector=vector / float!(A:Vector=vector - vector!(A:Vector=@n2, B:Vector=@n1) -> (ReturnValue:Vector) {is=title K2Node_CallFunction}, B:double=Get Delta Time X()) -> (ReturnValue:Vector)) -> (Output_Get:Vector)
   > n4 = Set CarrierOffsetRateSmooth(CarrierOffsetRateSmooth:Vector=VInterpTo(Current=Get CarrierOffsetRateSmooth() -> (CarrierOffsetRateSmooth:Vector), Target=@n3, DeltaTime=Get Delta Time X(), InterpSpeed=3)) -> (Output_Get:Vector)
   n5 = BreakVector(InVec=@n4)
   n6 = float + float(A=float * float(A=@n5.Y, B=0.4), B=float * float(A=@n5.Z, B=-0.3)) {is=title K2Node_CommutativeAssociativeBinaryOperator}
   n7 = Get Is Playing Melee Anim()
   n8 = SelectFloat(A=float * float(A=@n6, B=0.5), B=@n6, bPickA=@n7)
   n9 = float * float(A=@n5, B=-0.4)
   n10 = "Clamp (Float)"(Value=SelectFloat(A=float * float(A=@n9, B=0.5), B=@n9, bPickA=@n7), Min=-30, Max=30)
   > n11 = Set CarrierOffsetRotationReaction(CarrierOffsetRotationReaction:Rotator=MakeRotator(Roll="Clamp (Float)"(Value=@n8, Min=-30, Max=30), Pitch=@n10, Yaw=@n10)) -> (Output_Get:Rotator)
   > Set CarrierOffsetRotationReactionSmooth(CarrierOffsetRotationReactionSmooth:Rotator=RInterpTo(Current=Get CarrierOffsetRotationReactionSmooth(), Target=@n11, DeltaTime=Get Delta Time X(), InterpSpeed=5)) -> (Output_Get:Rotator)
   > Outputs()
  ]
 }
 Convert Local Velocity to rotation value {
  path=EventGraph.K2Node_Composite_6.Convert Local Velocity to rotation value
  nodes [
   Inputs()
   > Set PlayerYawDeltaInterp(PlayerYawDeltaInterp:double=FInterpTo(
     Current=Get PlayerYawDeltaInterp() -> (PlayerYawDeltaInterp:double)
     Target=Get Yaw Delta Current Tick() -> (Yaw Delta Current Tick:double)
     DeltaTime=Get Delta Time X()
     InterpSpeed=2
    )) -> (Output_Get:double)
   > n1 = Set LocalVelocityInterp(LocalVelocityInterp:Vector=VInterpTo(
     Current=Get LocalVelocityInterp() -> (LocalVelocityInterp:Vector)
     Target=Get LocalVelocity() -> (LocalVelocity:Vector)
     DeltaTime=Get Delta Time X()
     InterpSpeed=2
    )) -> (Output_Get:Vector)
   n2 = Get PlayerYawDeltaInterp()
   n3 = BreakVector(InVec=@n1)
   n4 = float * float(A=@n3, B=0.05)
   n5 = float + float(A=float * float(A=@n3.Y, B=0.05), B=@n2)
   > Set CarrierOffsetRotationVelocity(CarrierOffsetRotationVelocity:Rotator=MakeRotator(
     Roll="Clamp (Float)"(Value=@n4, Min=-20, Max=20)
     Pitch="Clamp (Float)"(Value=@n5, Min=-20, Max=20)
     Yaw="Clamp (Float)"(Value=@n2, Min=-20, Max=20)
    )) -> (Output_Get:Rotator)
   > Set CarrierOffsetRotationVelocityAlpha(CarrierOffsetRotationVelocityAlpha:double=float - float(B=GetCurveValue(CurveName=CancelVelocityRotation)) {is=title K2Node_CallFunction}) -> (Output_Get:double)
   > Outputs()
   Get Fort Player Pawn()
   Get RootYawOffset_TurnInPlace_Smooth()
  ]
 }
 Skydive Dive Pitch {
  path=EventGraph.K2Node_Composite_7.Skydive Dive Pitch
  nodes [
   Inputs()
   n1 = BreakRotator(InRot=GetBaseAimRotation!(self:Pawn Object Reference=Get Fort Player Pawn()) -> (ReturnValue:Rotator))
   n2 = Get Skydiving Is Diving() -> (Skydiving Is Diving:bool)
   > n3 = Set BaseAimPitchSmooth(BaseAimPitchSmooth=FInterpTo(Current=Get BaseAimPitchSmooth(), Target=SelectFloat(A=@n1.Pitch, B=0, bPickA=@n2), DeltaTime=Get Delta Time X(), InterpSpeed=4))
   > Set SkydiveDivingRotationOffset(SkydiveDivingRotationOffset:Rotator=MakeRotator(Roll=float * float(A=@n3, B=-1))) -> (Output_Get:Rotator)
   > Outputs()
  ]
 }
 SetBackPackTargetAlpha {
  path=EventGraph.K2Node_Composite_8.SetBackPackTargetAlpha
  nodes [
   Sequence!(execute) -> (then_0, then_1, then_2:Exec, then_3:Exec)
   > n1 = Branch(Condition=Get Is DBNO())
   > SetBackPackTargetVectorAlpha0() {is=title K2Node_CallFunction}
   n2 = Branch(execute=@n1.else, Condition=Get Is Playing Melee Anim())
   > SetBackPackTargetVectorAlpha1() {is=title K2Node_CallFunction}
   n3 = Branch(execute=@n2.else, Condition=Get Recently Fired() -> (Recently Fired:bool))
   > Set ForceBackPackTargetVectorAlpha1FromRecentlyFired(ForceBackPackTargetVectorAlpha1FromRecentlyFired:bool=true) -> (Output_Get:bool)
   > RetriggerableDelay(Duration=0.5)
   > Set ForceBackPackTargetVectorAlpha1FromRecentlyFired(ForceBackPackTargetVectorAlpha1FromRecentlyFired=false)
   n4 = Branch(
    execute=@n3.else
    Condition=Get ForceBackPackTargetVectorAlpha1FromRecentlyFired() -> (ForceBackPackTargetVectorAlpha1FromRecentlyFired:bool)
   )
   > SetBackPackTargetVectorAlpha1()
   n5 = Sequence!(execute=@n4.else) -> (then_0, then_1)
   > n6 = Branch(Condition=Get Is Relaxed Level 1() -> (Is Relaxed Level 1:bool))
   > Do Once(Reset=[@n10.then, @n6.else])
   > SetBackPackTargetVectorAlpha1()
   n7 = NOT Boolean(A=Get Emote_NonPlayerTrigger Montage Playing() -> (Emote_NonPlayerTrigger Montage Playing:bool))
   n8 = Get AnyEmotePlaying() -> (AnyEmotePlaying:bool)
   n9 = Branch(execute=@n5.then_1, Condition=OR Boolean(A=Get Is Moving 2D(), B=Get Is Jumping(), C=AND Boolean(A=@n8, B=@n7)))
   > SetBackPackTargetVectorAlpha0()
   n10 = SetBackPackTargetVectorAlpha1(execute=@n9.else)
   Inputs()
   > n11 = Sequence!(execute) -> (then_0, then_1, then_2:Exec, then_3:Exec)
   > n12 = Branch(Condition=Get Recently Fired())
   > Set ForceBackPackTargetVectorAlpha1FromRecentlyFired(ForceBackPackTargetVectorAlpha1FromRecentlyFired=true)
   > RetriggerableDelay(Duration=0.5)
   > Set ForceBackPackTargetVectorAlpha1FromRecentlyFired(ForceBackPackTargetVectorAlpha1FromRecentlyFired=false)
   n13 = NOT Boolean(A=Get Is Moving 2D())
   n14 = Get Is Relaxed Level 1()
   n15 = Branch(
    execute=@n12.else
    Condition=OR Boolean(A=Get ForceBackPackTargetVectorAlpha1FromRecentlyFired(), B=Get Is Playing Melee Anim(), C=AND Boolean(A=@n14, B=@n13))
   )
   > Set BackPackTargetVectorAlpha(BackPackTargetVectorAlpha=SelectFloat(A=0.3, B=0.6, bPickA=Get Is Playing Melee Anim()))
   n16 = NOT Boolean(A=Get Emote_NonPlayerTrigger Montage Playing())
   n17 = Get AnyEmotePlaying()
   Branch(execute=@n15.else, Condition=OR Boolean(A=Get Is Moving 2D(), B=Get Is Jumping(), C=AND Boolean(A=@n17, B=@n16)))
   > Set BackPackTargetVectorAlpha(BackPackTargetVectorAlpha=0)
   Outputs(then=@n11.then_3)
   PrintString(InString="To String (Boolean)"(InBool=Get Is Relaxed Level 1()))
   Branch(Condition=Get Is Relaxed Level 1())
  ]
 }
 TriggerMeleeEventsForSpringChange {
  path=EventGraph.K2Node_Composite_9.TriggerMeleeEventsForSpringChange
  nodes [
   Inputs()
   > n1 = Sequence!(execute) -> (then_0, then_1)
   > n2 = Branch(Condition=Get Is Playing Melee Anim())
   > n3 = Sequence()
   > Do Once(Reset=@n4.then_1)
   > MeleeStart() {is=title K2Node_CallFunction}
   Do Once(execute=@n4.then_0, Reset=@n3.then_1)
   > MeleeStop() {is=title K2Node_CallFunction}
   n4 = Sequence(execute=@n2.else)
   Outputs!(then_1:Exec=@n1.then_1)
  ]
 }
 GetOffsetValues {nodes [
   Inputs!()
   n1 = Break Pet Offsets Invader!(PetOffsets_Invader:"Pet Offsets Invader Structure (by ref)"=Get CarrierOffsets() -> (CarrierOffsets:Pet Offsets Invader Structure)) -> (
    DefaultIdle_8_#13:Vector=~
    DefaultDownsights_14_#14:Vector=~
    CrouchIdle_13_#15:Vector=~
    CrouchDownsights_15_#16:Vector=~
    DBNOIdle_9_#17:Vector=~
    MeleePlaying_17_#18:Vector=~
    CrouchMeleePlaying_51_#19:Vector=~
    EmotePlaying_19_#20:Vector=~
    DefaultSkydive_22_#21:Vector=~
    ParachuteOpen_53_#22:Vector=~
    DivingSkydive_48_#23:Vector=~
    SurfaceSwimmingIdle_26_#24:Vector=~
    SurfaceSwimmingDownsights_28_#25:Vector=~
    SurfaceSwimmingMeleePlaying_32_#26:Vector=~
    Surface_Swimming_Sprint_56_#27:Vector=~
    DiveJumping_35_#28:Vector=~
    CrouchTurning_44_#29:Vector=~
    PlayingBreakAnim_79_#30:Vector=~
    IsOperatingTurret_82_#31:Vector=~
    YawDefaultIdle_59_#32:float=~
    YawMeleePlaying_64_#33:float=~
    YawDefaultSkydive_65_#34:float=~
    YawParachuteOpen_66_#35:float=~
    YawDivingSkydive_67_#36:float=~
    YawEmotePlaying_70_#37:float=~
    YawDBNO_73_#38:float=~
    YawPlayingBreakAnim_78_#39:float=~
   ) {is=title K2Node_BreakStruct}
   n2 = vector + vector!(
    A=SelectVector(A=@n1.DefaultDownsights_14_#14, B=@n1, bPickA=Get Is Targeting() -> (Is Targeting:bool))
    B=Get Jump Up Offset Smooth()
   ) -> (ReturnValue)
   n3 = Get Transition Locomotion Additive Crouch Turning() -> (Transition Locomotion Additive Crouch Turning:bool)
   n4 = Get Is Moving 2D()
   n5 = vector + vector(
    A=SelectVector(A=@n1.CrouchDownsights_15_#16, B=@n1.CrouchIdle_13_#15, bPickA=Get Is Targeting())
    B=SelectVector(A=@n1.CrouchTurning_44_#29, B="0, 0, 0", bPickA=OR Boolean!(A=@n3, B=@n4) -> (ReturnValue))
   )
   n6 = Get Is Crouching()
   n7 = SelectVector(
    A=@n1.MeleePlaying_17_#18
    B=SelectVector(A=@n1.DBNOIdle_9_#17, B=SelectVector(A=@n5, B=@n2, bPickA=@n6), bPickA=Get Is DBNO())
    bPickA=Get Is Playing Melee Anim()
   )
   n8 = AND Boolean(A=Get Is Playing Melee Anim(), B=Get Is Crouching())
   n9 = NOT Boolean(A=Get Emote_NonPlayerTrigger Montage Playing())
   n10 = Get AnyEmotePlaying()
   n11 = SelectVector(
    A=SelectVector(
     A=@n1.ParachuteOpen_53_#22
     B=@n1.DefaultSkydive_22_#21
     bPickA=Get Is Parachute Open()
    )
    B=SelectVector(
     A=@n1.EmotePlaying_19_#20
     B=SelectVector(A=@n1.CrouchMeleePlaying_51_#19, B=@n7, bPickA=@n8)
     bPickA=AND Boolean(A=@n10, B=@n9)
    )
    bPickA=Get Is Skydiving() -> (Is Skydiving:bool)
   )
   n12 = Get Is Diving() -> (Is Diving:bool)
   n13 = SelectVector(
    A=@n1.SurfaceSwimmingMeleePlaying_32_#26
    B=SelectVector(
     A=@n1.SurfaceSwimmingDownsights_28_#25
     B=@n1.SurfaceSwimmingIdle_26_#24
     bPickA=Get Is Targeting()
    )
    bPickA=Get Is Playing Melee Anim()
   )
   n14 = Get Is Sprinting() -> (Is Sprinting:bool)
   n15 = SelectVector(
    A=@n1.DiveJumping_35_#28
    B=SelectVector(
     A=SelectVector(A=@n1.Surface_Swimming_Sprint_56_#27, B=@n13, bPickA=@n14)
     B=SelectVector(A=@n1.DivingSkydive_48_#23, B=@n11, bPickA=@n12)
     bPickA=Get Is Surface Swimming()
    )
    bPickA=Get Is Dive Jumping() -> (Is Dive Jumping:bool)
   )
   n16 = Get IsPlayingBreakAnim() -> (IsPlayingBreakAnim:bool)
   n17 = Get Fort Player Pawn()
   Outputs!(ReturnValue:Vector=SelectVector(
     A=@n1.IsOperatingTurret_82_#31
     B=SelectVector(A=@n1.PlayingBreakAnim_79_#30, B=@n15, bPickA=@n16)
     bPickA=IsOperatingTurret!(self:Fort Player Pawn Object Reference=@n17) -> (ReturnValue:bool)
    ))
  ]}
 GetRotationOffsetValues {nodes [
   Inputs()
   n1 = Break Pet Offsets Invader(PetOffsets_Invader=Get CarrierOffsets()) -> (
    DefaultIdle_8_#13=~
    DefaultDownsights_14_#14=~
    CrouchIdle_13_#15=~
    CrouchDownsights_15_#16=~
    DBNOIdle_9_#17=~
    MeleePlaying_17_#18=~
    CrouchMeleePlaying_51_#19=~
    EmotePlaying_19_#20=~
    DefaultSkydive_22_#21=~
    ParachuteOpen_53_#22=~
    DivingSkydive_48_#23=~
    SurfaceSwimmingIdle_26_#24=~
    SurfaceSwimmingDownsights_28_#25=~
    SurfaceSwimmingMeleePlaying_32_#26=~
    Surface_Swimming_Sprint_56_#27=~
    DiveJumping_35_#28=~
    CrouchTurning_44_#29=~
    PlayingBreakAnim_79_#30=~
    IsOperatingTurret_82_#31=~
    YawDefaultIdle_59_#32=~
    YawMeleePlaying_64_#33=~
    YawDefaultSkydive_65_#34=~
    YawParachuteOpen_66_#35=~
    YawDivingSkydive_67_#36=~
    YawEmotePlaying_70_#37=~
    YawDBNO_73_#38=~
    YawPlayingBreakAnim_78_#39=~
   )
   n2 = Get Emote_NonPlayerTrigger Montage Playing()
   n3 = SelectFloat(
    A=@n1.YawEmotePlaying_70_#37
    B=SelectFloat(A=@n1.YawDBNO_73_#38, B=@n1.YawDefaultIdle_59_#32, bPickA=Get Is DBNO())
    bPickA=AND Boolean(A=Get AnyEmotePlaying(), B=NOT Boolean(A=@n2))
   )
   n4 = Get Skydiving Is Skydiving()
   Outputs(ReturnValue:double=SelectFloat(
     A=@n1.YawPlayingBreakAnim_78_#39
     B=SelectFloat(A=@n1.YawDefaultSkydive_65_#34, B=@n3, bPickA=@n4)
     bPickA=Get IsPlayingBreakAnim()
    ))
  ]}
]
variables [
 Allow Locomotion Break {type=boolean}
 AnyEmotePlaying {type=boolean}
 As B Pet Inst Invader {type=B_PetInst_Invader_C, default=null}
 BackPackTargetVector {type=Vector}
 BackPackTargetVectorAlpha {type=number}
 BackPackTargetVectorAlphaSmooth {type=number}
 BaseAimPitchSmooth {type=number}
 Carrier Offset Spring State {type=VectorSpringState, default {}}
 Carrier Offset Total Last Tick {type=Vector}
 Carrier Root Rotation Spring State {type=FloatSpringState, default {}}
 Carrier Rotation Spring State {type=FloatSpringState, default {}}
 CarrierOffsetRateSmooth {type=Vector}
 CarrierOffsetRotationReaction {type=Rotator}
 CarrierOffsetRotationReactionSmooth {type=Rotator}
 CarrierOffsetRotationVelocity {type=Rotator}
 CarrierOffsetRotationVelocityAlpha {type=number}
 CarrierOffsets {
  type=PetOffsets_Invader
  default {
   yaw DBNO=100
   yaw Default Idle=10
   yaw Default Skydive=0
   yaw Diving Skydive=0
   yaw Emote Playing=65
   yaw Melee Playing=0
   yaw Parachute Open=0
   yaw PlayingBreakAnim=45
   crouch Downsights=<20 -45 -30>
   crouch Idle=<15 -40 -20>
   crouch Melee Playing=<0 -60 -40>
   crouch Turning=<5 -10 15>
   dBNO Idle=<40 -15 -25>
   default Downsights=<20 -45 0>
   default Idle=<10 -40 35>
   default Skydive=<0 -30 40>
   dive Jumping=<30 -5 35>
   diving Skydive=<0 -20 50>
   emote Playing=<60 -40 32>
   isOperatingTurret=<0 -80 -25>
   melee Playing=<7 -35 20>
   parachute Open=<0 -80 20>
   playingBreakAnim=<5 -50 32>
   surface Swimming Downsights=<25 -35 30>
   surface Swimming Idle=<15 -35 25>
   surface Swimming Melee Playing=<0 -50 -20>
   surface_Swimming_Sprint=<30 -30 40>
  }
 }
 CarrierOffsetTotal {type=Vector}
 CarrierPackOffsetRate {type=Vector}
 Choose Loco Break Index {type=integer}
 Delta Time X {type=number}
 Emote_NonPlayerTrigger Montage {type=AnimMontage, default=/Game/Animation/Game/MainPlayer/Emotes/NonPlayerTrigger/Emote_NonPlayerTrigger_CMM_M}
 Emote_NonPlayerTrigger Montage Playing {type=boolean}
 ForceBackPackTargetVectorAlpha1FromRecentlyFired {type=boolean}
 Fort Player AnimBP {type=Fortnite_M_Avg_Player_AnimBlueprint_C, default=null}
 Fort Player Pawn {type=FortPlayerPawn}
 FrontEndAnimBP {type=FrontendAnimInstance, default=null}
 HoverRotationOffsetTotal {type=Rotator}
 Is Crouching {type=boolean}
 Is DBNO {type=boolean}
 Is DBNOCarried {type=boolean}
 Is Dive Jumping {type=boolean}
 Is Diving {type=boolean}
 Is Jumping {type=boolean}
 Is Moving 2D {type=boolean}
 Is Parachute Just Opened {type=boolean}
 Is Parachute Open {type=boolean}
 Is Parachute Open Last Tick {type=boolean}
 Is Playing Emote {type=boolean}
 Is Playing Melee Anim {type=boolean}
 Is Relaxed Level 1 {type=boolean}
 Is Skydiving {type=boolean}
 Is Sprinting {type=boolean}
 Is Surface Swimming {type=boolean}
 Is Targeting {type=boolean}
 IsDBNOCarriedAlpha {type=number}
 IsMindContollingPlayer {type=boolean, default=true}
 IsMovingOnGround {type=boolean}
 IsPlayingBreakAnim {type=boolean}
 Jump Up Offset Smooth {type=Vector}
 Local Velocity Yaw Angle {type=number}
 LocalVelocity {type=Vector}
 LocalVelocityInterp {type=Vector}
 Melee Twist Curve Value {type=number}
 Parachute Just Opened Offset {type=Vector}
 PlayerYawDeltaInterp {type=number}
 Recently Fired {type=boolean}
 Root Yaw Offset {type=number}
 RootRotationMeleeTwistSpring {type=number}
 RootRotationOffsetTotal {type=Rotator}
 RootYawOffset_TurnInPlace_Smooth {type=number}
 RotationYawDefinedHoverOffset {type=Rotator}
 ScaleExperimentation {type=Vector, default=<1 1 1>}
 SkydiveDivingRotationOffset {type=Rotator}
 Skydiving Is Diving {type=boolean}
 Skydiving Is Parachute Open {type=boolean}
 Skydiving Is Skydive Dive Mode {type=boolean}
 Skydiving Is Skydiving {type=boolean}
 SparkOn {type=boolean}
 SpringCriticalDampingFactor {type=number, default=0.5}
 SpringCriticalDampingFactorSmooth {type=number}
 SpringStiffness {type=number, default=10}
 SpringStiffnessInterpSpeed {type=number, default=5}
 SpringStiffnessInterpSpeedSmooth {type=number}
 SpringStiffnessSmooth {type=number}
 Transition Locomotion Additive Crouch Turning {type=boolean}
 Velocity Z {type=number}
 Yaw Delta Current Tick {type=number}
 YawDefinedHoverOffset {type=number}
]
